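# Control Rig module for a two-bone IK/FK limb, with helper functions for solving and matching controls.
# The IK solve feeds segment-scaled lengths and an exponentially softened effector into Two Bone IK, with optional stretch.
# Pole vector helpers derive transforms from root, IK control, twist follow and side.
# Matching functions snap FK and IK controls to virtual bones, resetting gimbal controls, and debug drawing shows softness radii and axes.
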ControlRigBlueprint /EpicControlRig/Modules/IkSolves/IkFk2Bones/CRM_Epic_IkFk2Bones_v02
graphs [
 ContainedGraph {nodes [
   n1 = RigVMFunctionEntryNode()
   > n2 = RigVMDispatchNode(Array=RigVMVariableNode() {variable=Scaled Lengths}.Value) {template="DISPATCH_RigVMDispatch_ArrayReset(io Array)"}
   RigVMFunction_Sequence = &RigVMFunction_Sequence(ExecuteContext=@n2.ExecuteContext)
   Make_Array = RigVMDispatchNode(
    Values=(0,0)
    Values.0=RigVMVariableNode() {variable=Upper Segment Scale}.Value
    Values.1=RigVMVariableNode() {variable=Lower Segment Scale}.Value
   ) {template="DISPATCH_RigVMDispatch_ArrayMake(in Values,out Array)"}
   RigVMVariableNode(Value=@Make_Array.Array, ExecuteContext=@RigVMFunction_Sequence.A) {variable=Scale Factors}
   > For_Each = RigVMDispatchNode(Array=RigVMVariableNode() {variable=Lengths}.Value) {template="DISPATCH_RigVMDispatch_ArrayIterator(in Array,out Element,out Index,out Count,out Ratio)"}
   n3 = RigVMVariableNode() {variable=Scale Factors}
   > RigVMDispatchNode(
    Element=&RigVMFunction_MathDoubleMul(
     B=RigVMDispatchNode(Index=@For_Each.Index, Array=@n3.Value) {template="DISPATCH_RigVMDispatch_ArrayGetAtIndex(in Array,in Index,out Element)"}.Element
     A=@For_Each.Element
    ).Result
    Array=RigVMVariableNode() {variable=Scaled Lengths}.Value
   ) {template="DISPATCH_RigVMDispatch_ArrayAdd(io Array,in Element,out Index)"}
   n4 = RigVMVariableNode() {variable=Scaled Lengths}
   n5 = RigVMDispatchNode(Index=0, Array=@n4.Value) {template="DISPATCH_RigVMDispatch_ArrayGetAtIndex(in Array,in Index,out Element)"}
   n6 = RigVMVariableNode() {variable=Upper Segment Scale}
   n7 = &RigUnit_GetTransform(Item=@n1.IK Control, bInitial=@n1.Initial)
   n8 = &RigUnit_GetTransform(bInitial=@n1.Initial, Item=@n1.Root Bone)
   Interpolate = /Script/RigVM.RigVMFunction_MathVectorLerp(
    T=&RigVMFunction_MathDoubleDiv(
     B=/Script/RigVM.RigVMFunction_MathDoubleArraySum(Array=@n4.Value) {is=struct RigVMUnitNode, template="ArraySum::Execute(in Array,out Sum)"}.Sum
     A=&RigVMFunction_MathDoubleMul(A=@n5.Element, B=@n6.Value).Result
    ).Result
    B=@n7.Transform.Translation
    A=@n8.Transform.Translation
   ) {is=struct RigVMUnitNode, template="Interpolate::Execute(in A,in B,in T,out Result)"}
   n9 = RigVMVariableNode() {variable=IK Control}
   n10 = /Script/ControlRig.RigUnit_AimBoneMath(
    DebugSettings=(Scale=10,WorldOffset=(Rotation=(X=0,Y=0,Z=0,W=1),Translation=(X=0,Y=0,Z=0),Scale3D=(X=1,Y=1,Z=1)))
    InputTransform=(Rotation=(X=0,Y=0,Z=0,W=1),Translation=(X=0,Y=0,Z=0),Scale3D=(X=1,Y=1,Z=1))
    Primary=(Weight=1,Axis=(X=1,Y=0,Z=0),Target=(X=0,Y=0,Z=0),Kind=Location,Space=(Type=Bone))
    Secondary=(Weight=1,Axis=(X=0,Y=0,Z=1),Target=(X=0,Y=0,Z=0),Kind=Direction,Space=(Type=Control,Name="IK"))
    Weight=1
    Secondary.Target.Z=RigVMDispatchNode(False=1, True=-1, Condition=RigVMVariableNode() {variable=Negative Side}.Value) {template="DISPATCH_RigVMDispatch_If(in Condition,in True,in False,out Result)"}.Result
    InputTransform.Translation=@Interpolate.Result
    Primary.Space=@n9.Value
    Secondary.Space=@n9.Value
   ) {is=struct RigVMUnitNode}
   n11 = &RigUnit_AimBoneMath(
    DebugSettings=(Scale=10,WorldOffset=(Rotation=(X=0,Y=0,Z=0,W=1),Translation=(X=0,Y=0,Z=0),Scale3D=(X=1,Y=1,Z=1)))
    InputTransform=(Rotation=(X=0,Y=0,Z=0,W=1),Translation=(X=0,Y=0,Z=0),Scale3D=(X=1,Y=1,Z=1))
    Primary=(Weight=1,Axis=(X=1,Y=0,Z=0),Target=(X=0,Y=0,Z=0),Kind=Location,Space=(Type=Bone))
    Secondary=(Weight=1,Axis=(X=0,Y=1,Z=0),Target=(X=0,Y=1,Z=0),Kind=Direction,Space=(Type=Bone))
    Weight=1
    InputTransform.Translation=@Interpolate.Result
    Primary.Space=@n9.Value
    Secondary.Space=RigVMVariableNode() {variable=Root Bone}.Value
   )
   n12 = /Script/RigVM.RigVMFunction_MathTransformLerp(B=@n10.Result, A=@n11.Result, T=RigVMVariableNode() {variable=PV Twist Follow}.Value) {is=struct RigVMUnitNode, template="Interpolate::Execute(in A,in B,in T,out Result)"}
   n13 = RigVMVariableNode() {variable=Scaled Lengths}
   DISPATCH_RigVMDispatch_ArrayGetAtIndex = RigVMDispatchNode(Index=0, Array=@n13.Value) {template="DISPATCH_RigVMDispatch_ArrayGetAtIndex(in Array,in Index,out Element)"}
   ArraySum = &RigVMFunction_MathDoubleArraySum(Array=@n13.Value) {template="ArraySum::Execute(in Array,out Sum)"}
   RigVMFunction_MathVectorDistance = &RigVMFunction_MathVectorDistance(B=@n7.Transform.Translation, A=@n8.Transform.Translation)
   Divide = &RigVMFunction_MathDoubleDiv(B=@ArraySum.Sum, A=@RigVMFunction_MathVectorDistance.Result)
   Multiply = &RigVMFunction_MathDoubleMul(B=@DISPATCH_RigVMDispatch_ArrayGetAtIndex.Element, A=@Divide.Result)
   Subtract = &RigVMFunction_MathDoubleSub(A=1, B=@Divide.Result)
   Add = &RigVMFunction_MathDoubleAdd(A=@Multiply.Result, B=&RigVMFunction_MathDoubleMul(B=@ArraySum.Sum, A=@Subtract.Result).Result)
   Less = /Script/RigVM.RigVMFunction_MathDoubleLess(B=0, A=@Subtract.Result) {is=struct RigVMUnitNode, template="Less::Execute(in A,in B,out Result)"}
   n14 = RigVMDispatchNode(False=@Add.Result, True=@DISPATCH_RigVMDispatch_ArrayGetAtIndex.Element, Condition=@Less.Result) {template="DISPATCH_RigVMDispatch_If(in Condition,in True,in False,out Result)"}
   Negate = /Script/RigVM.RigVMFunction_MathDoubleNegate(Value=@n14.Result) {is=struct RigVMUnitNode, template="Negate::Execute(in Value,out Result)"}
   n15 = /Script/RigVM.RigVMFunction_MathTransformMakeAbsolute(
    Local=(Rotation=(X=0,Y=0,Z=0,W=1),Translation=(X=0,Y=0,Z=0),Scale3D=(X=1,Y=1,Z=1))
    Local.Translation.Y=RigVMDispatchNode(True=@n14.Result, False=@Negate.Result, Condition=RigVMVariableNode() {variable=Negative Side}.Value) {template="DISPATCH_RigVMDispatch_If(in Condition,in True,in False,out Result)"}.Result
    Parent=@n12.Result
   ) {is=struct RigVMUnitNode, template="Make Absolute::Execute(in Local,in Parent,out Global)"}
   Make_Array_1 = RigVMDispatchNode(
    Values=((Rotation=(X=0,Y=0,Z=0,W=1),Translation=(X=0,Y=0,Z=0),Scale3D=(X=1,Y=1,Z=1)),((Rotation=(X=0,Y=0,Z=0,W=1),Translation=(X=0,Y=0,Z=0),Scale3D=(X=1,Y=1,Z=1)),Rotation=(X=0,Y=0,Z=0,W=1),Translation=(X=0,Y=0,Z=0),Scale3D=(X=1,Y=1,Z=1)))
    Values.0=@n12.Result
    Values.1=@n15.Global
   ) {template="DISPATCH_RigVMDispatch_ArrayMake(in Values,out Array)"}
   RigVMFunction_DebugTransformArrayMutableNoSpace = /Script/RigVM.RigVMFunction_DebugTransformArrayMutableNoSpace(
    Color=(R=1,G=1,B=1,A=1)
    DebugDrawSettings=(DepthPriority=SDPG_Foreground,Lifetime=-1)
    Mode=Axes
    ParentIndices=()
    Scale=10
    Thickness=0
    WorldOffset=(Rotation=(X=0,Y=0,Z=0,W=1),Translation=(X=0,Y=0,Z=0),Scale3D=(X=1,Y=1,Z=1))
    Transforms=@Make_Array_1.Array
    bEnabled=RigVMVariableNode() {variable=Debug}.Value
    ExecutePin=@RigVMFunction_Sequence.B
   ) {is=struct RigVMUnitNode}
   > RigVMFunctionReturnNode(Result=@n15.Global)
   RigVMVariableNode() {variable=Lengths}
  ]}
 ContainedGraph {nodes [
   n1 = RigVMFunctionEntryNode()
   SequenceAggregate = &RigVMFunction_Sequence(ExecuteContext=@n1.ExecuteContext)
   DISPATCH_RigVMDispatch_ArrayReset = RigVMDispatchNode(ExecuteContext=@SequenceAggregate.A, Array=RigVMVariableNode() {variable=Values}.Value) {template="DISPATCH_RigVMDispatch_ArrayReset(io Array)"}
   n2 = RigVMVariableNode() {variable=Items}
   > DISPATCH_RigVMDispatch_ArrayIterator = RigVMDispatchNode(Array=@n2.Value) {template="DISPATCH_RigVMDispatch_ArrayIterator(in Array,out Element,out Index,out Count,out Ratio)"}
   DISPATCH_RigVMDispatch_CoreEquals = RigVMDispatchNode(B=0, A=@DISPATCH_RigVMDispatch_ArrayIterator.Index) {template="DISPATCH_RigVMDispatch_CoreEquals(in A,in B,out Result)"}
   RigVMFunction_ControlFlowBranch = &RigVMFunction_ControlFlowBranch(ExecuteContext=@DISPATCH_RigVMDispatch_ArrayIterator.ExecuteContext, Condition=@DISPATCH_RigVMDispatch_CoreEquals.Result)
   n3 = RigVMVariableNode() {variable=Initial}
   Subtract = /Script/RigVM.RigVMFunction_MathIntSub(B=1, A=@DISPATCH_RigVMDispatch_ArrayIterator.Index) {is=struct RigVMUnitNode, template="Subtract::Execute(in A,in B,out Result)"}
   DISPATCH_RigVMDispatch_ArrayGetAtIndex = RigVMDispatchNode(Index=@Subtract.Result, Array=@n2.Value) {template="DISPATCH_RigVMDispatch_ArrayGetAtIndex(in Array,in Index,out Element)"}
   RigVMFunction_MathVectorDistance = &RigVMFunction_MathVectorDistance(
    A=&RigUnit_GetTransform(Item=@DISPATCH_RigVMDispatch_ArrayIterator.Element, bInitial=@n3.Value).Transform.Translation
    B=&RigUnit_GetTransform(Item=@DISPATCH_RigVMDispatch_ArrayGetAtIndex.Element, bInitial=@n3.Value).Transform.Translation
   )
   DISPATCH_RigVMDispatch_ArrayAdd = RigVMDispatchNode(
    Array=@DISPATCH_RigVMDispatch_ArrayReset.Array
    ExecuteContext=@RigVMFunction_ControlFlowBranch.False
    Element=@RigVMFunction_MathVectorDistance.Result
   ) {template="DISPATCH_RigVMDispatch_ArrayAdd(io Array,in Element,out Index)"}
   RigVMFunctionReturnNode(Distances=@DISPATCH_RigVMDispatch_ArrayReset.Array, ExecuteContext=@SequenceAggregate.B)
   RigVMVariableNode() {variable=Values}
   RigVMVariableNode() {variable=Values}
  ]}
 ContainedGraph {nodes [
   n1 = RigVMFunctionEntryNode()
   n2 = /Script/RigVM.RigVMFunction_MathVectorSub(B=@n1.Root, A=@n1.Target) {is=struct RigVMUnitNode, template="Subtract::Execute(in A,in B,out Result)"}
   n3 = /Script/RigVM.RigVMFunction_MathVectorUnit(Value=RigVMDispatchNode(
     Condition=@n1.Negate
     True=/Script/RigVM.RigVMFunction_MathVectorNegate(Value=@n2.Result) {is=struct RigVMUnitNode, template="Negate::Execute(in Value,out Result)"}.Result
     False=@n2.Result
    ) {template="DISPATCH_RigVMDispatch_If(in Condition,in True,in False,out Result)"}.Result) {is=struct RigVMUnitNode, template="Unit::Execute(in Value,out Result)"}
   n4 = /Script/RigVM.RigVMFunction_MathVectorScale(
    Factor=/Script/RigVM.RigVMFunction_MathFloatDiv(
     B=@n1.Length
     A=/Script/RigVM.RigVMFunction_MathVectorLength(Value=@n2.Result) {is=struct RigVMUnitNode, template="Length::Execute(in Value,out Result)"}.Result
    ) {is=struct RigVMUnitNode, template="Divide::Execute(in A,in B,out Result)"}.Result
    Value=@n3.Result
   ) {is=struct RigVMUnitNode, template="Scale::Execute(in Value,in Factor,out Result)"}
   n5 = RigVMDispatchNode(True=@n4.Result, False=@n3.Result, Condition=@n1.Apply Scale) {template="DISPATCH_RigVMDispatch_If(in Condition,in True,in False,out Result)"}
   n6 = /Script/RigVM.RigVMFunction_MathMatrixToVectors(Value=/Script/RigVM.RigVMFunction_MathMatrixFromTransformV2(Value=@n1.Root Tfm) {is=struct RigVMUnitNode, template="Cast::Execute(in Value,out Result)"}.Result) {is=struct RigVMUnitNode}
   n7 = &RigVMFunction_MathVectorUnit(Value=/Script/RigVM.RigVMFunction_MathVectorCross(A=@n6.Z, B=@n3.Result) {is=struct RigVMUnitNode, template="Cross::Execute(in A,in B,out Result)"}.Result) {template="Unit::Execute(in Value,out Result)"}
   n8 = &RigVMFunction_MathVectorScale(Factor=&RigVMFunction_MathVectorLength(Value=@n6.Y) {template="Length::Execute(in Value,out Result)"}.Result, Value=@n7.Result) {template="Scale::Execute(in Value,in Factor,out Result)"}
   n9 = RigVMDispatchNode(True=@n8.Result, False=@n7.Result, Condition=@n1.Apply Scale) {template="DISPATCH_RigVMDispatch_If(in Condition,in True,in False,out Result)"}
   n10 = &RigVMFunction_MathVectorUnit(Value=&RigVMFunction_MathVectorCross(B=@n8.Result, A=@n4.Result) {template="Cross::Execute(in A,in B,out Result)"}.Result) {template="Unit::Execute(in Value,out Result)"}
   n11 = RigVMDispatchNode(
    True=&RigVMFunction_MathVectorScale(Factor=&RigVMFunction_MathVectorLength(Value=@n6.Z) {template="Length::Execute(in Value,out Result)"}.Result, Value=@n10.Result) {template="Scale::Execute(in Value,in Factor,out Result)"}.Result
    False=@n10.Result
    Condition=@n1.Apply Scale
   ) {template="DISPATCH_RigVMDispatch_If(in Condition,in True,in False,out Result)"}
   RigVMFunctionReturnNode(Result=/Script/RigVM.RigVMFunction_MathMatrixToTransform(Value=/Script/RigVM.RigVMFunction_MathMatrixFromVectors(Origin=@n1.Root, X=@n5.Result, Y=@n9.Result, Z=@n11.Result) {is=struct RigVMUnitNode}.Result) {is=struct RigVMUnitNode, template="Cast::Execute(in Value,out Result)"}.Result)
  ]}
 New Function_ContainedGraph {nodes [
   n1 = RigVMFunctionEntryNode()
   n2 = RigVMVariableNode() {variable=Distance}
   Negate = &RigVMFunction_MathDoubleNegate(Value=@n2.Value) {template="Negate::Execute(in Value,out Result)"}
   DISPATCH_RigVMDispatch_If = RigVMDispatchNode(False=@Negate.Result, True=@n2.Value, Condition=RigVMVariableNode() {variable=Negative Side}.Value) {template="DISPATCH_RigVMDispatch_If(in Condition,in True,in False,out Result)"}
   Multiply = &RigVMFunction_MathDoubleMul(A=@DISPATCH_RigVMDispatch_If.Result, B=RigVMVariableNode() {variable=Distance Scale}.Value)
   n3 = /Script/ControlRig.RigUnit_GetTransformItemArray(Space=GlobalSpace, bInitial=False, Items=@n1.Items) {is=struct RigVMUnitNode}
   n4 = RigVMDispatchNode(Index=1, Array=@n3.Transforms) {template="DISPATCH_RigVMDispatch_ArrayGetAtIndex(in Array,in Index,out Element)"}
   Interpolate = /Script/RigVM.RigVMFunction_MathQuaternionSlerp(
    T=0.5
    A=RigVMDispatchNode(Index=0, Array=@n3.Transforms) {template="DISPATCH_RigVMDispatch_ArrayGetAtIndex(in Array,in Index,out Element)"}.Element.Rotation
    B=@n4.Element.Rotation
   ) {is=struct RigVMUnitNode, template="Interpolate::Execute(in A,in B,in T,out Result)"}
   RigVMFunction_MathTransformMake = /Script/RigVM.RigVMFunction_MathTransformMake(Scale=(X=1,Y=1,Z=1), Translation=@n4.Element.Translation, Rotation=@Interpolate.Result) {is=struct RigVMUnitNode}
   RigVMFunction_MathTransformTransformVector = /Script/RigVM.RigVMFunction_MathTransformTransformVector(Location=(X=0,Y=0,Z=0), Location.Y=@Multiply.Result, Transform=@RigVMFunction_MathTransformMake.Result) {is=struct RigVMUnitNode}
   > RigVMFunctionReturnNode(Result=@RigVMFunction_MathTransformTransformVector.Result)
  ]}
 New Function_ContainedGraph {nodes [
   n1 = RigVMFunctionEntryNode()
   n2 = &RigUnit_GetTransformItemArray(Space=GlobalSpace, bInitial=False, Items=@n1.Items)
   n3 = RigVMDispatchNode(Index=1, Array=@n2.Transforms) {template="DISPATCH_RigVMDispatch_ArrayGetAtIndex(in Array,in Index,out Element)"}
   Interpolate = &RigVMFunction_MathQuaternionSlerp(
    T=0.5
    A=RigVMDispatchNode(Index=0, Array=@n2.Transforms) {template="DISPATCH_RigVMDispatch_ArrayGetAtIndex(in Array,in Index,out Element)"}.Element.Rotation
    B=@n3.Element.Rotation
   ) {template="Interpolate::Execute(in A,in B,in T,out Result)"}
   DISPATCH_RigVMDispatch_MakeStruct = RigVMDispatchNode(
    Elements=(Rotation=(X=0,Y=0,Z=0,W=1),Translation=(X=0,Y=0,Z=0),Scale3D=(X=1,Y=1,Z=1))
    Elements.Translation=@n3.Element.Translation
    Elements.Rotation=@Interpolate.Result
   ) {template="DISPATCH_RigVMDispatch_MakeStruct(in Elements,out Struct)"}
   > RigVMFunctionReturnNode(Struct=@DISPATCH_RigVMDispatch_MakeStruct.Struct)
  ]}
 New Function_ContainedGraph {nodes [
   n1 = RigVMFunctionEntryNode()
   RigVMFunction_MathBoolToInteger = &RigVMFunction_MathBoolToInteger(Value=@n1.Value)
   DISPATCH_RigVMDispatch_SelectInt32 = RigVMDispatchNode(Values=("Sphere_Solid","Box_Solid"), Values.0=@n1.False Name, Values.1=@n1.True Name, Index=@RigVMFunction_MathBoolToInteger.Result) {template="DISPATCH_RigVMDispatch_SelectInt32(in Index,in Values,out Result)"}
   n2 = /Script/ControlRig.RigUnit_HierarchyGetShapeSettings(Item=@n1.Item) {is=struct RigVMUnitNode}
   > /Script/ControlRig.RigUnit_HierarchySetShapeSettings(
    Settings=(bVisible=True,Name="Default",Color=(R=1,G=0,B=0,A=1),Transform=(Rotation=(X=0,Y=0,Z=0,W=1),Translation=(X=0,Y=0,Z=0),Scale3D=(X=1,Y=1,Z=1)))
    Settings.Name=@DISPATCH_RigVMDispatch_SelectInt32.Result
    Item=@n1.Item
    Settings.Color=@n2.Settings.Color
    Settings.Transform=@n2.Settings.Transform
    Settings.bVisible=@n2.Settings.bVisible
   ) {is=struct RigVMUnitNode}
   > RigVMFunctionReturnNode()
  ]}
 New Function_ContainedGraph {nodes [
   n1 = RigVMFunctionEntryNode()
   RigVMFunction_Sequence = &RigVMFunction_Sequence(ExecuteContext=@n1.ExecuteContext)
   RigVMVariableNode(
    Value=&RigUnit_GetTransform(bInitial=False, Item=RigVMVariableNode() {variable=Start Location Item}.Value).Transform.Translation
    ExecuteContext=@RigVMFunction_Sequence.A
   ) {variable=Start Location}
   > RigVMVariableNode(Value=&RigUnit_GetTransform(bInitial=False, Item=RigVMVariableNode() {variable=End Location Item}.Value).Transform.Translation) {variable=End Location}
   > RigVMVariableNode(Value=&RigUnit_GetTransform(bInitial=False, Item=RigVMVariableNode() {variable=Start Orientation Item}.Value).Transform.Rotation) {variable=Start Orientation}
   > RigVMVariableNode(Value=&RigUnit_GetTransform(bInitial=False, Item=RigVMVariableNode() {variable=End Orientation Item}.Value).Transform.Rotation) {variable=End Orientation}
   n2 = RigVMVariableNode() {variable=Start Target Axis}
   DISPATCH_RigVMDispatch_CoreEquals = RigVMDispatchNode(B=(X=0,Y=1,Z=0), A=@n2.Value) {template="DISPATCH_RigVMDispatch_CoreEquals(in A,in B,out Result)"}
   RigVMFunction_MathBoolAnd = &RigVMFunction_MathBoolAnd(A=@DISPATCH_RigVMDispatch_CoreEquals.Result, B=RigVMVariableNode() {variable=Condition}.Value)
   n3 = RigVMVariableNode() {variable=Start Location}
   n4 = RigVMVariableNode() {variable=End Location}
   n5 = &RigUnit_AimBoneMath(
    DebugSettings=(bEnabled=False,Scale=10,WorldOffset=(Rotation=(X=0,Y=0,Z=0,W=1),Translation=(X=0,Y=0,Z=0),Scale3D=(X=1,Y=1,Z=1)))
    InputTransform=(Rotation=(X=0,Y=0,Z=0,W=1),Translation=(X=0,Y=0,Z=0),Scale3D=(X=1,Y=1,Z=1))
    Primary=(Weight=1,Axis=(X=1,Y=0,Z=0),Target=(X=1,Y=0,Z=0),Kind=Location,Space=(Type=Bone,Name="None"))
    Secondary=(Weight=1,Axis=(X=0,Y=0,Z=1),Target=(X=1,Y=0,Z=0),Kind=Direction,Space=(Type=Bone,Name="None"))
    Weight=1
    Secondary.Axis=RigVMDispatchNode(
     True=&RigVMFunction_MathVectorNegate(Value=@n2.Value) {template="Negate::Execute(in Value,out Result)"}.Result
     Condition=@RigVMFunction_MathBoolAnd.Result
     False=@n2.Value
    ) {template="DISPATCH_RigVMDispatch_If(in Condition,in True,in False,out Result)"}.Result
    Secondary.Target=/Script/RigVM.RigVMFunction_MathQuaternionRotateVector(Transform=RigVMVariableNode() {variable=Start Orientation}.Value, Vector=@n2.Value) {is=struct RigVMUnitNode, template="Rotate Vector::Execute(in Transform,in Vector,out Result)"}.Result
    InputTransform.Translation=@n3.Value
    Primary.Target=@n4.Value
   )
   n6 = RigVMVariableNode() {variable=End Target Axis}
   Negate = &RigVMFunction_MathVectorNegate(Value=@n6.Value) {template="Negate::Execute(in Value,out Result)"}
   DISPATCH_RigVMDispatch_If = RigVMDispatchNode(True=@Negate.Result, False=@n6.Value, Condition=RigVMVariableNode() {variable=Condition}.Value) {template="DISPATCH_RigVMDispatch_If(in Condition,in True,in False,out Result)"}
   n7 = &RigUnit_AimBoneMath(
    DebugSettings=(bEnabled=False,Scale=10,WorldOffset=(Rotation=(X=0,Y=0,Z=0,W=1),Translation=(X=0,Y=0,Z=0),Scale3D=(X=1,Y=1,Z=1)))
    InputTransform=(Rotation=(X=0,Y=0,Z=0,W=1),Translation=(X=0,Y=0,Z=0),Scale3D=(X=1,Y=1,Z=1))
    Primary=(Weight=1,Axis=(X=-1,Y=0,Z=0),Target=(X=1,Y=0,Z=0),Kind=Location,Space=(Type=Bone,Name="None"))
    Secondary=(Weight=1,Axis=(X=0,Y=0,Z=1),Target=(X=1,Y=0,Z=0),Kind=Direction,Space=(Type=Bone,Name="None"))
    Weight=1
    Secondary.Axis=@DISPATCH_RigVMDispatch_If.Result
    Secondary.Target=&RigVMFunction_MathQuaternionRotateVector(Transform=RigVMVariableNode() {variable=End Orientation}.Value, Vector=@n6.Value) {template="Rotate Vector::Execute(in Transform,in Vector,out Result)"}.Result
    Primary.Target=@n3.Value
    InputTransform.Translation=@n4.Value
   )
   RigVMFunctionReturnNode(Start Transform=@n5.Result, End Transform=@n7.Result, ExecuteContext=@RigVMFunction_Sequence.B)
  ]}
 New Function_ContainedGraph {nodes [
   RigVMFunctionEntryNode()
   > RigVMFunction_Sequence = RigVMAggregateNode()
   n1 = &RigUnit_GetTransform(bInitial=true, Item=RigVMVariableNode() {variable=Bone B}.Value)
   n2 = &RigVMFunction_MathVectorDistance(
    B=@n1.Transform.Translation
    A=&RigUnit_GetTransform(bInitial=true, Item=RigVMVariableNode() {variable=Bone A}.Value).Transform.Translation
   )
   RigVMVariableNode(Value=@n2.Result, ExecuteContext=@RigVMFunction_Sequence.A) {variable=Bone A Initial Length}
   n3 = &RigVMFunction_MathVectorDistance(
    A=@n1.Transform.Translation
    B=&RigUnit_GetTransform(bInitial=true, Item=RigVMVariableNode() {variable=Bone C}.Value).Transform.Translation
   )
   > RigVMVariableNode(Value=@n3.Result) {variable=Bone B Initial Length}
   Add = &RigVMFunction_MathDoubleAdd(
    A=&RigVMFunction_MathDoubleMul(A=@n2.Result, B=RigVMVariableNode() {variable=Upper Segment Scale}.Value).Result
    B=&RigVMFunction_MathDoubleMul(A=@n3.Result, B=RigVMVariableNode() {variable=Lower Segment Scale}.Value).Result
   )
   > RigVMVariableNode(Value=@Add.Result) {variable=Chain Initial Length}
   n4 = &RigVMFunction_MathVectorSub(
    B=&RigUnit_GetTransform(bInitial=False, Item=RigVMVariableNode() {variable=Bone A}.Value).Transform.Translation
    A=RigVMVariableNode() {variable=Effector}.Value.Translation
   ) {template="Subtract::Execute(in A,in B,out Result)"}
   RigVMFunction_MathVectorLength = &RigVMFunction_MathVectorLength(Value=@n4.Result) {template="Length::Execute(in Value,out Result)"}
   > RigVMVariableNode(Value=@RigVMFunction_MathVectorLength.Result) {variable=Current Distance}
   > RigVMVariableNode(Value=@n4.Result) {variable=Target}
   n5 = RigVMVariableNode() {variable=Chain Initial Length}
   Multiply = &RigVMFunction_MathDoubleMul(B=RigVMVariableNode() {variable=Softness Distance}.Value, A=@n5.Value)
   Remap = &RigVMFunction_MathDoubleRemap(
    SourceMaximum=1
    SourceMinimum=0
    bClamp=False
    TargetMaximum=@Multiply.Result
    TargetMinimum=@n5.Value
    Value=RigVMVariableNode() {variable=Softness}.Value
   )
   > RigVMVariableNode(Value=@Remap.Result) {variable=Softness Start Distance}
   RigVMFunction_ControlFlowBranch = &RigVMFunction_ControlFlowBranch(ExecuteContext=@RigVMFunction_Sequence.B, Condition=RigVMVariableNode() {variable=Debug}.Value)
   n6 = &RigUnit_GetTransform(bInitial=False, Item=RigVMVariableNode() {variable=Bone A}.Value)
   RigVMFunction_DebugArcNoSpace = /Script/RigVM.RigVMFunction_DebugArcNoSpace(
    Color=(R=0,G=1,B=0,A=1)
    DebugDrawSettings=(DepthPriority=SDPG_Foreground,Lifetime=-1)
    Detail=128
    MaximumDegrees=360
    MinimumDegrees=0
    Thickness=0
    WorldOffset=(Rotation=(X=0,Y=0,Z=0,W=1),Translation=(X=0,Y=0,Z=0),Scale3D=(X=1,Y=1,Z=1))
    bEnabled=True
    Transform=@n6.Transform
    ExecutePin=@RigVMFunction_ControlFlowBranch.True
    Radius=RigVMVariableNode() {variable=Softness Start Distance}.Value
   ) {is=struct RigVMUnitNode}
   > &RigVMFunction_DebugArcNoSpace(
    Color=(R=1,G=0,B=0,A=1)
    DebugDrawSettings=(DepthPriority=SDPG_Foreground,Lifetime=-1)
    Detail=128
    MaximumDegrees=360
    MinimumDegrees=0
    Thickness=0
    WorldOffset=(Rotation=(X=0,Y=0,Z=0,W=1),Translation=(X=0,Y=0,Z=0),Scale3D=(X=1,Y=1,Z=1))
    bEnabled=True
    Transform=@n6.Transform
    Radius=RigVMVariableNode() {variable=Chain Initial Length}.Value
   )
   n7 = &RigVMFunction_MathDoubleSub(A=RigVMVariableNode() {variable=Current Distance}.Value, B=RigVMVariableNode() {variable=Softness Start Distance}.Value)
   IsNearlyZero = /Script/RigVM.RigVMFunction_MathDoubleIsNearlyZero(Tolerance=0.001, Value=@n7.Result) {is=struct RigVMUnitNode, template="IsNearlyZero::Execute(in Value,in Tolerance,out Result)"}
   n8 = &RigVMFunction_MathDoubleSub(B=RigVMVariableNode() {variable=Softness Start Distance}.Value, A=RigVMVariableNode() {variable=Chain Initial Length}.Value)
   Divide = &RigVMFunction_MathDoubleDiv(
    A=RigVMDispatchNode(True=1, Condition=@IsNearlyZero.Result, False=@n7.Result) {template="DISPATCH_RigVMDispatch_If(in Condition,in True,in False,out Result)"}.Result
    B=RigVMDispatchNode(
     True=1
     Condition=&RigVMFunction_MathDoubleIsNearlyZero(Tolerance=0.001, Value=@n8.Result) {template="IsNearlyZero::Execute(in Value,in Tolerance,out Result)"}.Result
     False=@n8.Result
    ) {template="DISPATCH_RigVMDispatch_If(in Condition,in True,in False,out Result)"}.Result
   )
   Negate = &RigVMFunction_MathDoubleNegate(Value=@Divide.Result) {template="Negate::Execute(in Value,out Result)"}
   n9 = &RigVMFunction_MathDoubleSub(
    A=1
    B=/Script/RigVM.RigVMFunction_MathDoubleExponential(Value=@Negate.Result) {is=struct RigVMUnitNode, template="Exponential::Execute(in Value,out Result)"}.Result
   )
   LessEqual = /Script/RigVM.RigVMFunction_MathDoubleLessEqual(B=0, A=@n9.Result) {is=struct RigVMUnitNode, template="LessEqual::Execute(in A,in B,out Result)"}
   RigVMFunction_MathBoolOr = &RigVMFunction_MathBoolOr(
    B=@LessEqual.Result
    A=&RigVMFunction_MathDoubleLessEqual(B=0.0001, A=RigVMVariableNode() {variable=Softness}.Value) {template="LessEqual::Execute(in A,in B,out Result)"}.Result
   )
   n10 = &RigVMFunction_ControlFlowBranch(Condition=@RigVMFunction_MathBoolOr.Result, ExecuteContext=@RigVMFunction_Sequence.C)
   n11 = &RigVMFunction_MathDoubleAdd(
    B=&RigVMFunction_MathDoubleMul(A=@n9.Result, B=@n8.Result).Result
    A=RigVMVariableNode() {variable=Softness Start Distance}.Value
   )
   n12 = &RigVMFunction_MathDoubleDiv(B=@n11.Result, A=RigVMVariableNode() {variable=Current Distance}.Value)
   n13 = RigVMVariableNode() {variable=Bone A Initial Length}
   n14 = RigVMVariableNode() {variable=Upper Segment Scale}
   RigVMVariableNode(
    Value=&RigVMFunction_MathDoubleMul(B=@n12.Result, A=&RigVMFunction_MathDoubleMul(A=@n13.Value, B=@n14.Value).Result).Result
    ExecuteContext=@n10.False
   ) {variable=Soft Length A}
   n15 = RigVMVariableNode() {variable=Bone B Initial Length}
   n16 = RigVMVariableNode() {variable=Lower Segment Scale}
   > RigVMVariableNode(Value=&RigVMFunction_MathDoubleMul(B=@n12.Result, A=&RigVMFunction_MathDoubleMul(A=@n15.Value, B=@n16.Value).Result).Result) {variable=Soft Length B}
   n17 = RigVMVariableNode() {variable=Bone A}
   RigVMFunction_MathVectorSetLength = /Script/RigVM.RigVMFunction_MathVectorSetLength(Length=@n11.Result, Value=RigVMVariableNode() {variable=Target}.Value) {is=struct RigVMUnitNode, template="SetLength::Execute(in Value,in Length,out Result)"}
   n18 = RigVMVariableNode() {variable=Effector}
   > RigVMVariableNode(
    Value=(Rotation=(X=0,Y=0,Z=0,W=1),Translation=(X=0,Y=0,Z=0),Scale3D=(X=1,Y=1,Z=1))
    Value.Translation=/Script/RigVM.RigVMFunction_MathVectorAdd(
     A=&RigUnit_GetTransform(bInitial=False, Item=@n17.Value).Transform.Translation
     B=@RigVMFunction_MathVectorSetLength.Result
    ) {is=struct RigVMUnitNode, template="Add::Execute(in A,in B,out Result)"}.Result
    Value.Rotation=@n18.Value.Rotation
    Value.Scale3D=@n18.Value.Scale3D
   ) {variable=Soft Effector}
   RigVMVariableNode(
    Value=&RigVMFunction_MathDoubleMul(A=RigVMVariableNode() {variable=Bone A Initial Length}.Value, B=RigVMVariableNode() {variable=Upper Segment Scale}.Value).Result
    ExecuteContext=@n10.True
   ) {variable=Soft Length A}
   > RigVMVariableNode(Value=&RigVMFunction_MathDoubleMul(A=RigVMVariableNode() {variable=Bone B Initial Length}.Value, B=RigVMVariableNode() {variable=Lower Segment Scale}.Value).Result) {variable=Soft Length B}
   > RigVMVariableNode(Value=RigVMVariableNode() {variable=Effector}.Value) {variable=Soft Effector}
   DISPATCH_RigVMDispatch_If = RigVMDispatchNode(
    True=RigVMVariableNode() {variable=Effector}.Value
    Condition=RigVMVariableNode() {variable=EnableStretch}.Value
    False=RigVMVariableNode() {variable=Soft Effector}.Value
   ) {template="DISPATCH_RigVMDispatch_If(in Condition,in True,in False,out Result)"}
   n19 = RigVMVariableNode() {variable=EnableStretch}
   Inverse = /Script/RigVM.RigVMFunction_MathQuaternionInverse(Value=RigVMVariableNode() {variable=IK Rotation Offset}.Value) {is=struct RigVMUnitNode, template="Inverse::Execute(in Value,out Result)"}
   /Script/ControlRig.RigUnit_TwoBoneIKSimplePerItem(
    DebugSettings=(bEnabled=False,Scale=10,WorldOffset=(Rotation=(X=0,Y=0,Z=0,W=1),Translation=(X=0,Y=0,Z=0),Scale3D=(X=1,Y=1,Z=1)))
    Effector=(Rotation=(X=0,Y=0,Z=0,W=1),Translation=(X=0,Y=0,Z=0),Scale3D=(X=1,Y=1,Z=1))
    bPropagateToChildren=True
    Effector.Scale3D=@DISPATCH_RigVMDispatch_If.Result.Scale3D
    Effector.Translation=@DISPATCH_RigVMDispatch_If.Result.Translation
    ItemALength=RigVMDispatchNode(False=0, True=RigVMVariableNode() {variable=Soft Length A}.Value, Condition=@n19.Value) {template="DISPATCH_RigVMDispatch_If(in Condition,in True,in False,out Result)"}.Result
    ItemBLength=RigVMDispatchNode(False=0, True=RigVMVariableNode() {variable=Soft Length B}.Value, Condition=@n19.Value) {template="DISPATCH_RigVMDispatch_If(in Condition,in True,in False,out Result)"}.Result
    Effector.Rotation=/Script/RigVM.RigVMFunction_MathQuaternionMul(A=@DISPATCH_RigVMDispatch_If.Result.Rotation, B=@Inverse.Result) {is=struct RigVMUnitNode, template="Multiply::Execute(in A,in B,out Result)"}.Result
    ExecutePin=@RigVMFunction_Sequence.D
    ItemA=RigVMVariableNode() {variable=Bone A}.Value
    ItemB=RigVMVariableNode() {variable=Bone B}.Value
    EffectorItem=RigVMVariableNode() {variable=Bone C}.Value
    PrimaryAxis=RigVMVariableNode() {variable=PrimaryAxis}.Value
    SecondaryAxis=RigVMVariableNode() {variable=SecondaryAxis}.Value
    SecondaryAxisWeight=RigVMVariableNode() {variable=SecondaryAxisWeight}.Value
    PoleVector=RigVMVariableNode() {variable=PoleVector}.Value
    PoleVectorKind=RigVMVariableNode() {variable=PoleVectorKind}.Value
    PoleVectorSpace=RigVMVariableNode() {variable=PoleVectorSpace}.Value
    bEnableStretch=RigVMVariableNode() {variable=EnableStretch}.Value
    StretchStartRatio=RigVMVariableNode() {variable=StretchStartRatio}.Value
    StretchMaximumRatio=RigVMVariableNode() {variable=StretchMaximumRatio}.Value
    Weight=RigVMVariableNode() {variable=Weight}.Value
   ) {is=struct RigVMUnitNode}
   > RigVMFunction_DebugTransformMutableNoSpace = /Script/RigVM.RigVMFunction_DebugTransformMutableNoSpace(
    Color=(R=1,G=0,B=0,A=1)
    DebugDrawSettings=(DepthPriority=SDPG_Foreground,Lifetime=-1)
    Mode=Axes
    Scale=20
    Thickness=1
    Transform=(Rotation=(X=0,Y=0,Z=0,W=1),Translation=(X=0,Y=0,Z=0),Scale3D=(X=1,Y=1,Z=1))
    WorldOffset=(Rotation=(X=0,Y=0,Z=0,W=1),Translation=(X=0,Y=0,Z=0),Scale3D=(X=1,Y=1,Z=1))
    bEnabled=RigVMVariableNode() {variable=Debug}.Value
   ) {is=struct RigVMUnitNode}
   RigVMFunctionReturnNode(ExecuteContext=@RigVMFunction_Sequence.E)
  ]}
 New Function_ContainedGraph {nodes [
   RigVMFunctionEntryNode()
   > RigVMFunction_Sequence = RigVMAggregateNode()
   RigVMFunction_ControlFlowBranch = &RigVMFunction_ControlFlowBranch(ExecuteContext=@RigVMFunction_Sequence.A, Condition=RigVMVariableNode() {variable=Debug}.Value)
   n1 = RigVMVariableNode() {variable=Bone Transforms}
   n2 = RigVMDispatchNode(Index=0, Array=@n1.Value) {template="DISPATCH_RigVMDispatch_ArrayGetAtIndex(in Array,in Index,out Element)"}
   n3 = RigVMDispatchNode(Index=1, Array=@n1.Value) {template="DISPATCH_RigVMDispatch_ArrayGetAtIndex(in Array,in Index,out Element)"}
   RigVMFunction_MathMatrixFromTransformV2 = &RigVMFunction_MathMatrixFromTransformV2(Value=@n3.Element) {template="Cast::Execute(in Value,out Result)"}
   RigVMFunction_MathMatrixToVectors = &RigVMFunction_MathMatrixToVectors(Value=@RigVMFunction_MathMatrixFromTransformV2.Result)
   n4 = &RigVMFunction_MathVectorCross(
    A=@RigVMFunction_MathMatrixToVectors.Y
    B=&RigVMFunction_MathVectorSub(
     B=@n2.Element.Translation
     A=RigVMDispatchNode(Index=2, Array=@n1.Value) {template="DISPATCH_RigVMDispatch_ArrayGetAtIndex(in Array,in Index,out Element)"}.Element.Translation
    ) {template="Subtract::Execute(in A,in B,out Result)"}.Result
   ) {template="Cross::Execute(in A,in B,out Result)"}
   RigVMFunction_MathIntersectPlane = /Script/RigVM.RigVMFunction_MathIntersectPlane(PlanePoint=@n2.Element.Translation, Start=@n3.Element.Translation, Direction=@n4.Result, PlaneNormal=@n4.Result) {is=struct RigVMUnitNode}
   Subtract = &RigVMFunction_MathVectorSub(B=@n2.Element.Translation, A=@RigVMFunction_MathIntersectPlane.Result) {template="Subtract::Execute(in A,in B,out Result)"}
   RigVMFunction_MathVectorLength = &RigVMFunction_MathVectorLength(Value=@Subtract.Result) {template="Length::Execute(in Value,out Result)"}
   Divide = &RigVMFunction_MathDoubleDiv(
    A=&RigVMFunction_MathVectorDistance(B=@n2.Element.Translation, A=@n3.Element.Translation).Result
    B=@RigVMFunction_MathVectorLength.Result
   )
   Multiply = /Script/RigVM.RigVMFunction_MathVectorMul(B=(X=1,Y=1,Z=1), B.X=@Divide.Result, B.Y=@Divide.Result, B.Z=@Divide.Result, A=@Subtract.Result) {is=struct RigVMUnitNode, template="Multiply::Execute(in A,in B,out Result)"}
   Add = &RigVMFunction_MathVectorAdd(A=@n2.Element.Translation, B=@Multiply.Result) {template="Add::Execute(in A,in B,out Result)"}
   RigVMFunction_DebugLineNoSpace = /Script/RigVM.RigVMFunction_DebugLineNoSpace(
    Color=(R=1,G=0,B=0,A=1)
    DebugDrawSettings=(DepthPriority=SDPG_Foreground,Lifetime=-1)
    Thickness=0.2
    WorldOffset=(Rotation=(X=0,Y=0,Z=0,W=1),Translation=(X=0,Y=0,Z=0),Scale3D=(X=1,Y=1,Z=1))
    bEnabled=True
    B=@Add.Result
    A=@n2.Element.Translation
    ExecutePin=@RigVMFunction_ControlFlowBranch.True
   ) {is=struct RigVMUnitNode}
   > RigVMFunction_DebugRectangleNoSpace = /Script/RigVM.RigVMFunction_DebugRectangleNoSpace(
    Color=(R=1,G=0,B=0,A=1)
    DebugDrawSettings=(DepthPriority=SDPG_Foreground,Lifetime=-1)
    Scale=1
    Thickness=2
    Transform=(Rotation=(X=0,Y=0,Z=0,W=1),Translation=(X=0,Y=0,Z=0),Scale3D=(X=1,Y=1,Z=1))
    WorldOffset=(Rotation=(X=0,Y=0,Z=0,W=1),Translation=(X=0,Y=0,Z=0),Scale3D=(X=1,Y=1,Z=1))
    bEnabled=True
    Transform.Translation=@RigVMFunction_MathIntersectPlane.Result
   ) {is=struct RigVMUnitNode}
   RigVMFunction_MathTransformMake = &RigVMFunction_MathTransformMake(Translation=@Add.Result, Rotation=@n3.Element.Rotation, Scale=@n3.Element.Scale3D)
   DISPATCH_RigVMDispatch_ArraySetAtIndex = RigVMDispatchNode(
    Index=1
    Element=@RigVMFunction_MathTransformMake.Result
    ExecuteContext=@RigVMFunction_Sequence.B
    Array=RigVMVariableNode() {variable=Bone Transforms}.Value
   ) {template="DISPATCH_RigVMDispatch_ArraySetAtIndex(io Array,in Index,in Element)"}
   > RigVMFunctionReturnNode(New Bone Transforms=@DISPATCH_RigVMDispatch_ArraySetAtIndex.Array)
  ]}
 New Function_ContainedGraph {nodes [
   n1 = RigVMFunctionEntryNode()
   RigVMFunction_Sequence = &RigVMFunction_Sequence(ExecuteContext=@n1.ExecuteContext)
   RigVMDispatchNode(ExecuteContext=@RigVMFunction_Sequence.A, Array=RigVMVariableNode() {variable=Transforms Local}.Value) {template="DISPATCH_RigVMDispatch_ArrayReset(io Array)"}
   > For_Each = RigVMDispatchNode(Array=@n1.Transforms Input) {template="DISPATCH_RigVMDispatch_ArrayIterator(in Array,out Element,out Index,out Count,out Ratio)"}
   RigVMFunctionReturnNode(Transforms Output=(), ExecuteContext=@RigVMFunction_Sequence.B)
   &RigVMFunction_MathQuaternionMul(B=@n1.Rotation Offset, A=@For_Each.Element.Rotation) {template="Multiply::Execute(in A,in B,out Result)"}
  ]}
 New Function_ContainedGraph {nodes [
   RigVMFunctionEntryNode()
   > For_Each_1_1 = RigVMDispatchNode(Array=&RigUnit_GetTransformItemArray(Space=GlobalSpace, Items=RigVMVariableNode() {variable=Virtual Bones}.Value).Transforms) {template="DISPATCH_RigVMDispatch_ArrayIterator(in Array,out Element,out Index,out Count,out Ratio)"}
   Not_Equals = RigVMDispatchNode(B=0, A=@For_Each_1_1.Index) {template="DISPATCH_RigVMDispatch_CoreNotEquals(in A,in B,out Result)"}
   RigVMFunction_ControlFlowBranch = &RigVMFunction_ControlFlowBranch(ExecuteContext=@For_Each_1_1.ExecuteContext, Condition=@Not_Equals.Result)
   n1 = RigVMDispatchNode(Index=@For_Each_1_1.Index, Array=RigVMVariableNode() {variable=FK Controls}.Value) {template="DISPATCH_RigVMDispatch_ArrayGetAtIndex(in Array,in Index,out Element)"}
   &RigUnit_SetTransform(
    Value=(Rotation=(X=0,Y=0,Z=0,W=1),Translation=(X=0,Y=0,Z=0),Scale3D=(X=1,Y=1,Z=1))
    Weight=1
    bInitial=False
    ExecutePin=@RigVMFunction_ControlFlowBranch.Completed
    Value.Rotation=&RigVMFunction_MathQuaternionMul(A=@For_Each_1_1.Element.Rotation, B=RigVMVariableNode() {variable=FK Rotation Offset}.Value) {template="Multiply::Execute(in A,in B,out Result)"}.Result
    Item=@n1.Element
    Value.Scale3D=@For_Each_1_1.Element.Scale3D
    Value.Translation=@For_Each_1_1.Element.Translation
   )
   n2 = /Script/ControlRig.RigUnit_HierarchyGetParent(bDefaultParent=True, Child=@n1.Element) {is=struct RigVMUnitNode}
   &RigUnit_SetTransform(
    Value=(Rotation=(X=0,Y=0,Z=0,W=1),Translation=(X=0,Y=0,Z=0),Scale3D=(X=1,Y=1,Z=1))
    Weight=1
    bInitial=False
    ExecutePin=@RigVMFunction_ControlFlowBranch.True
    Value.Scale3D=@For_Each_1_1.Element.Scale3D
    Item=@n2.Parent
    Value.Translation=@For_Each_1_1.Element.Translation
   )
   > Set Transform_2 = &RigUnit_SetRotation(Value=(X=0,Y=0,Z=0,W=1), Weight=1, bInitial=False, Item=@n2.Parent)
   For_Each = RigVMDispatchNode(ExecuteContext=@For_Each_1_1.Completed, Array=RigVMVariableNode() {variable=FK Gimbal Controls}.Value) {template="DISPATCH_RigVMDispatch_ArrayIterator(in Array,out Element,out Index,out Count,out Ratio)"}
   > &RigUnit_SetTransform(
    Space=LocalSpace
    Value=(Rotation=(X=0,Y=0,Z=0,W=1),Translation=(X=0,Y=0,Z=0),Scale3D=(X=1,Y=1,Z=1))
    Weight=1
    bInitial=False
    bPropagateToChildren=false
    Item=@For_Each.Element
   )
   RigVMFunctionReturnNode(ExecuteContext=@For_Each.Completed)
  ]}
 New Function_ContainedGraph {nodes [
   RigVMFunctionEntryNode()
   n1 = RigVMVariableNode() {variable=Virtual Bones}
   > &RigUnit_SetTransform(
    Weight=1
    bInitial=False
    Value=&RigUnit_GetTransform(
     bInitial=False
     Item=RigVMDispatchNode(Index=0, Array=@n1.Value) {template="DISPATCH_RigVMDispatch_ArrayGetAtIndex(in Array,in Index,out Element)"}.Element
    ).Transform
    Item=RigVMVariableNode() {variable=IK Base Control}.Value
   )
   n2 = &RigUnit_GetTransform(
    bInitial=False
    Item=RigVMDispatchNode(Index=-1, Array=RigVMVariableNode() {variable=Virtual Bones}.Value) {template="DISPATCH_RigVMDispatch_ArrayGetAtIndex(in Array,in Index,out Element)"}.Element
   )
   n3 = &RigVMFunction_MathQuaternionMul(A=@n2.Transform.Rotation, B=RigVMVariableNode() {variable=IK Rotation Offset}.Value) {template="Multiply::Execute(in A,in B,out Result)"}
   > &RigUnit_SetTransform(
    Value=(Rotation=(X=0,Y=0,Z=0,W=1),Translation=(X=0,Y=0,Z=0),Scale3D=(X=1,Y=1,Z=1))
    Weight=1
    bInitial=False
    Value.Scale3D=@n2.Transform.Scale3D
    Value.Translation=@n2.Transform.Translation
    Value.Rotation=@n3.Result
    Item=RigVMVariableNode() {variable=IK Control}.Value
   )
   > n4 = &RigUnit_SetTransform(
    Space=LocalSpace
    Value=(Rotation=(X=0,Y=0,Z=0,W=1),Translation=(X=0,Y=0,Z=0),Scale3D=(X=1,Y=1,Z=1))
    Weight=1
    bInitial=False
    Item=RigVMVariableNode() {variable=IK Gimbal Control}.Value
   )
   n5 = &RigVMFunction_ControlFlowBranch(
    Condition=/Script/ControlRig.RigUnit_GetBoolAnimationChannelFromItem(bInitial=False, Item=RigVMVariableNode() {variable=Hand Align Channel}.Value) {is=struct RigVMUnitNode, template="GetAnimationChannelFromItem::Execute(out Value,in Item,in bInitial)"}.Value
    ExecuteContext=@n4.ExecutePin
   )
   Compute Pole Vector Parent_1 = RigVMFunctionReferenceNode(
    Initial=False
    PV Twist Follow=/Script/ControlRig.RigUnit_GetFloatAnimationChannel(Channel=PV Twist Follow, bInitial=False, Control=RigVMVariableNode() {variable=IK Control}.Value.Name) {is=struct RigVMUnitNode, template="GetAnimationChannel::Execute(out Value,in Control,in Channel,in bInitial)"}.Value
    ExecuteContext=@n5.Completed
    IK Control=RigVMVariableNode() {variable=PV Twist Orient Space}.Value
    Lengths=RigVMVariableNode() {variable=Default Lengths}.Value
    Upper Segment Scale=RigVMVariableNode() {variable=Upper Segment Scale}.Value
    Negative Side=RigVMVariableNode() {variable=Negative Side}.Value
    Lower Segment Scale=RigVMVariableNode() {variable=Lower Segment Scale}.Value
    Root Bone=RigVMVariableNode() {variable=Start Twist PV}.Value
   )
   > &RigUnit_SetTransform(Weight=1, bInitial=False, Value=@"Compute Pole Vector Parent_1".Result, Item=RigVMVariableNode() {variable=Auto PV Parent}.Value)
   > Get Distances Between = RigVMFunctionReferenceNode(Initial=false, Items=RigVMVariableNode() {variable=Bones}.Value)
   Compute PV Position_1 = RigVMFunctionReferenceNode(
    Distance=RigVMDispatchNode(Index=0, Array=@"Get Distances Between".Distances) {template="DISPATCH_RigVMDispatch_ArrayGetAtIndex(in Array,in Index,out Element)"}.Element
    Negative Side=RigVMVariableNode() {variable=Negative Side}.Value
    Items=RigVMVariableNode() {variable=Virtual Bones}.Value
    Distance Scale=RigVMVariableNode() {variable=PV Distance Scale}.Value
   )
   > Set Transform_1 = /Script/ControlRig.RigUnit_SetTranslation(
    Space=GlobalSpace
    Weight=1
    bInitial=False
    bPropagateToChildren=True
    Value=@"Compute PV Position_1".Result
    Item=RigVMVariableNode() {variable=PV Control}.Value
   ) {is=struct RigVMUnitNode, template="Set Transform::Execute(in Item,in Space,in bInitial,in Value,in Weight,in bPropagateToChildren)"}
   > RigVMFunctionReturnNode()
   SetControlOffset = /Script/ControlRig.RigUnit_SetControlOffset(
    Offset=(Rotation=(X=0,Y=0,Z=0,W=1),Translation=(X=0,Y=0,Z=0),Scale3D=(X=1,Y=1,Z=1))
    Space=GlobalSpace
    Offset.Translation=&RigUnit_GetTransform(bInitial=False, Item=RigVMVariableNode() {variable=IK Control}.Value).Transform.Translation
    Offset.Rotation=&RigUnit_GetTransform(bInitial=False, Item=RigVMVariableNode() {variable=IK Rotation Null}.Value).Transform.Rotation
    Control=RigVMVariableNode() {variable=IK Rotation Control}.Value.Name
    ExecutePin=@n5.True
   ) {is=struct RigVMUnitNode, template="SetControlOffset::Execute(in Control,io Offset,in Space)"}
   > Set Transform_2 = &RigUnit_SetRotation(Space=GlobalSpace, Weight=1, bInitial=False, Value=@n3.Result, Item=RigVMVariableNode() {variable=IK Rotation Control}.Value)
   &RigUnit_GetTransform(
    bInitial=False
    Item=RigVMDispatchNode(Index=-1, Array=RigVMVariableNode() {variable=Virtual Bones}.Value) {template="DISPATCH_RigVMDispatch_ArrayGetAtIndex(in Array,in Index,out Element)"}.Element
   )
   RigVMVariableNode() {variable=Default Lengths}
  ]}
 New Function_ContainedGraph {nodes [
   n1 = RigVMFunctionEntryNode()
   Get Control Color From Metadata_1_1 = RigVMFunctionReferenceNode(Color Override=RigVMVariableNode() {variable=Color}.Value)
   RigVMFunction_MathBoolAnd = &RigVMFunction_MathBoolAnd(
    A=RigVMDispatchNode(B=@n1.Color Override.G, A=@n1.Color Override.R) {template="DISPATCH_RigVMDispatch_CoreEquals(in A,in B,out Result)"}.Result
    B=RigVMDispatchNode(B=@n1.Color Override.B, A=@n1.Color Override.G) {template="DISPATCH_RigVMDispatch_CoreEquals(in A,in B,out Result)"}.Result
   )
   > RigVMFunctionReturnNode(Result=RigVMDispatchNode(False=@n1.Color Override, True=@"Get Control Color From Metadata_1_1".Result, Condition=@RigVMFunction_MathBoolAnd.Result) {template="DISPATCH_RigVMDispatch_If(in Condition,in True,in False,out Result)"}.Result)
  ]}
 New Function_ContainedGraph {nodes [
   n1 = RigVMFunctionEntryNode()
   RigVMFunction_Sequence = &RigVMFunction_Sequence(ExecuteContext=@n1.ExecuteContext)
   Not_Equals_1 = RigVMDispatchNode(
    A=&RigUnit_GetBoolAnimationChannelFromItem(bInitial=False, Item=RigVMVariableNode() {variable=Hand Align Channel}.Value) {template="GetAnimationChannelFromItem::Execute(out Value,in Item,in bInitial)"}.Value
    B=RigVMVariableNode() {variable=Previous Hand Align}.Value
   ) {template="DISPATCH_RigVMDispatch_CoreNotEquals(in A,in B,out Result)"}
   n2 = &RigVMFunction_ControlFlowBranch(Condition=@Not_Equals_1.Result, ExecuteContext=@RigVMFunction_Sequence.A)
   n3 = &RigVMFunction_ControlFlowBranch(
    Condition=&RigUnit_GetBoolAnimationChannelFromItem(bInitial=False, Item=RigVMVariableNode() {variable=Hand Align Channel}.Value) {template="GetAnimationChannelFromItem::Execute(out Value,in Item,in bInitial)"}.Value
    ExecuteContext=@n2.Completed
   )
   n4 = &RigUnit_GetTransform(bInitial=False, Item=RigVMVariableNode() {variable=IK Control}.Value)
   SetControlOffset = &RigUnit_SetControlOffset(
    Offset=(Rotation=(X=0,Y=0,Z=0,W=1),Translation=(X=0,Y=0,Z=0),Scale3D=(X=1,Y=1,Z=1))
    Space=GlobalSpace
    Offset.Scale3D=@n4.Transform.Scale3D
    Offset.Translation=@n4.Transform.Translation
    Offset.Rotation=&RigUnit_GetTransform(bInitial=False, Item=RigVMVariableNode() {variable=IK Rotation Null}.Value).Transform.Rotation
    ExecutePin=@n3.True
    Control=RigVMVariableNode() {variable=IK Rotation Control}.Value.Name
   ) {template="SetControlOffset::Execute(in Control,io Offset,in Space)"}
   n5 = RigVMVariableNode() {variable=IK Rotation Control}
   > Set Transform_1 = &RigUnit_SetRotation(
    Space=GlobalSpace
    Weight=1
    bInitial=False
    Item=RigVMDispatchNode(Index=-1, Array=RigVMVariableNode() {variable=Virtual Bones}.Value) {template="DISPATCH_RigVMDispatch_ArrayGetAtIndex(in Array,in Index,out Element)"}.Element
    Value=&RigVMFunction_MathQuaternionMul(
     A=&RigUnit_GetTransform(bInitial=False, Item=@n5.Value).Transform.Rotation
     B=RigVMVariableNode() {variable=IK Rotation Offset}.Value
    ) {template="Multiply::Execute(in A,in B,out Result)"}.Result
   )
   n6 = RigVMVariableNode(
    Value=&RigUnit_GetBoolAnimationChannelFromItem(bInitial=False, Item=RigVMVariableNode() {variable=Hand Align Channel}.Value) {template="GetAnimationChannelFromItem::Execute(out Value,in Item,in bInitial)"}.Value
    ExecuteContext=@n2.True
   ) {variable=Previous Hand Align}
   n7 = &RigVMFunction_ControlFlowBranch(
    Condition=&RigUnit_GetBoolAnimationChannelFromItem(bInitial=False, Item=RigVMVariableNode() {variable=Hand Align Channel}.Value) {template="GetAnimationChannelFromItem::Execute(out Value,in Item,in bInitial)"}.Value
    ExecuteContext=@n6.ExecuteContext
   )
   n8 = &RigUnit_GetTransform(bInitial=False, Item=RigVMVariableNode() {variable=IK Gimbal Control}.Value)
   &RigUnit_SetControlOffset(
    Offset=(Rotation=(X=0,Y=0,Z=0,W=1),Translation=(X=0,Y=0,Z=0),Scale3D=(X=1,Y=1,Z=1))
    Space=GlobalSpace
    Offset.Scale3D=@n8.Transform.Scale3D
    Offset.Translation=@n8.Transform.Translation
    Offset.Rotation=&RigUnit_GetTransform(bInitial=False, Item=RigVMVariableNode() {variable=IK Rotation Null}.Value).Transform.Rotation
    Control=RigVMVariableNode() {variable=IK Rotation Control}.Value.Name
    ExecutePin=@n7.True
   ) {template="SetControlOffset::Execute(in Control,io Offset,in Space)"}
   > Set Transform_2 = &RigUnit_SetRotation(
    Space=GlobalSpace
    Weight=1
    bInitial=False
    Value=&RigUnit_GetTransform(bInitial=False, Item=RigVMVariableNode() {variable=IK Gimbal Control}.Value).Transform.Rotation
    Item=RigVMVariableNode() {variable=IK Rotation Control}.Value
   )
   > /Script/ControlRig.RigUnit_SendEvent(
    Event=RequestAutoKey
    OffsetInSeconds=0
    bEnable=True
    bOnlyDuringInteraction=true
    Item=RigVMVariableNode() {variable=IK Rotation Control}.Value
   ) {is=struct RigVMUnitNode, template="SendEvent::Execute(in Event,in Item,in OffsetInSeconds,in bEnable,in bOnlyDuringInteraction)"}
   Set Transform_5 = &RigUnit_SetRotation(
    Space=GlobalSpace
    Weight=1
    bInitial=False
    Value=&RigUnit_GetTransform(bInitial=False, Item=RigVMVariableNode() {variable=IK Rotation Control}.Value).Transform.Rotation
    Item=RigVMVariableNode() {variable=IK Control}.Value
    ExecutePin=@n7.False
   )
   > &RigUnit_SendEvent(Event=RequestAutoKey, OffsetInSeconds=0, bEnable=True, bOnlyDuringInteraction=true, Item=RigVMVariableNode() {variable=IK Control}.Value) {template="SendEvent::Execute(in Event,in Item,in OffsetInSeconds,in bEnable,in bOnlyDuringInteraction)"}
   n9 = RigVMVariableNode() {variable=IK Gimbal Control}
   > Set Transform = &RigUnit_SetRotation(Value=(X=0,Y=0,Z=0,W=1), Weight=1, bInitial=False, Item=@n9.Value)
   > &RigUnit_SendEvent(Event=RequestAutoKey, OffsetInSeconds=0, bEnable=True, bOnlyDuringInteraction=true, Item=@n9.Value) {template="SendEvent::Execute(in Event,in Item,in OffsetInSeconds,in bEnable,in bOnlyDuringInteraction)"}
   RigVMFunctionReturnNode(ExecuteContext=@RigVMFunction_Sequence.B)
   RigVMVariableNode() {variable=IK Control}
   RigVMVariableNode() {variable=IK Control}
  ]}
 New Function_ContainedGraph {nodes [
   n1 = RigVMFunctionEntryNode()
   > RigVMVariableNode(Value=RigVMDispatchNode(Array=@n1.Transforms) {template="DISPATCH_RigVMDispatch_ArrayClone(in Array,out Clone)"}.Clone) {variable=Output}
   > n2 = RigVMDispatchNode(Array=@n1.Length) {template="DISPATCH_RigVMDispatch_ArrayIterator(in Array,out Element,out Index,out Count,out Ratio)"}
   n3 = RigVMDispatchNode(Index=@n2.Index, Array=@n1.Transforms) {template="DISPATCH_RigVMDispatch_ArrayGetAtIndex(in Array,in Index,out Element)"}
   FK Bone = RigVMCollapseNode(
    Root Tfm=@n3.Element
    Root=@n3.Element.Translation
    Target=RigVMDispatchNode(
     Index=/Script/RigVM.RigVMFunction_MathIntAdd(B=1, A=@n2.Index) {is=struct RigVMUnitNode, template="Add::Execute(in A,in B,out Result)"}.Result
     Array=@n1.Transforms
    ) {template="DISPATCH_RigVMDispatch_ArrayGetAtIndex(in Array,in Index,out Element)"}.Element.Translation
    Length=@n2.Element
    Negate=RigVMVariableNode() {variable=Negative Side}.Value
    Apply Scale=RigVMVariableNode() {variable=Apply Scale}.Value
   )
   > RigVMDispatchNode(Index=@n2.Index, Element=@"FK Bone".Result, Array=RigVMVariableNode() {variable=Output}.Value) {template="DISPATCH_RigVMDispatch_ArraySetAtIndex(io Array,in Index,in Element)"}
   RigVMFunctionReturnNode(ExecuteContext=@n2.Completed, Result=RigVMVariableNode() {variable=Output}.Value)
  ]}
 New Function_ContainedGraph {nodes [
   n1 = RigVMFunctionEntryNode()
   RigVMFunction_Sequence = &RigVMFunction_Sequence(ExecuteContext=@n1.ExecuteContext)
   For_Each_6 = RigVMDispatchNode(ExecuteContext=@RigVMFunction_Sequence.A, Array=RigVMVariableNode() {variable=Current Lengths}.Value) {template="DISPATCH_RigVMDispatch_ArrayIterator(in Array,out Element,out Index,out Count,out Ratio)"}
   n2 = RigVMVariableNode() {variable=Default Lengths}
   > /Script/ControlRig.RigUnit_SetFloatAnimationChannelFromItem(
    bInitial=False
    Item=RigVMDispatchNode(Index=@For_Each_6.Index, Array=RigVMVariableNode() {variable=Segment Scale Channels}.Value) {template="DISPATCH_RigVMDispatch_ArrayGetAtIndex(in Array,in Index,out Element)"}.Element
    Value=&RigVMFunction_MathDoubleDiv(
     B=RigVMDispatchNode(Index=@For_Each_6.Index, Array=@n2.Value) {template="DISPATCH_RigVMDispatch_ArrayGetAtIndex(in Array,in Index,out Element)"}.Element
     A=@For_Each_6.Element
    ).Result
   ) {is=struct RigVMUnitNode, template="SetAnimationChannelFromItem::Execute(in Value,in Item,in bInitial)"}
   RigVMFunctionReturnNode(ExecuteContext=@RigVMFunction_Sequence.B)
  ]}
 RigVMFunctionLibrary {nodes [
   "Calculate And Set Segment Scale Channel  Values" = RigVMCollapseNode(Current Lengths=(), Default Lengths=(), ExecuteContext=(), Segment Scale Channels=())
   Compute FK = RigVMCollapseNode(ExecuteContext=(), Length=(), Transforms=())
   Compute Mid Bone Transform For IK Plane Align = RigVMCollapseNode(Bone Transforms=(), Debug=false, ExecuteContext=())
   Compute Mid Null Transform = RigVMCollapseNode(Items=())
   Compute PV Position = RigVMCollapseNode(Distance=-20, Distance Scale=1, Items=(), Negative Side=False)
   Compute Pole Vector Parent = RigVMCollapseNode(
    ExecuteContext=()
    IK Control=(Type=Control,Name="foot_l_ik_ctrl")
    Initial=False
    Lengths=()
    Lower Segment Scale=1
    Negative Side=False
    Root Bone=(Type=Bone,Name="thigh_l")
    Upper Segment Scale=1
   )
   Compute Twist Sockets = RigVMCollapseNode(
    Condition=False
    End Location Item=(Type=Bone,Name="None")
    End Orientation Item=(Type=Bone,Name="None")
    End Target Axis=(X=0,Y=1,Z=0)
    ExecuteContext=()
    Start Location Item=(Type=Bone,Name="None")
    Start Orientation Item=(Type=Bone,Name="None")
    Start Target Axis=(X=0,Y=1,Z=0)
   )
   Control Color Override From Metadata = RigVMCollapseNode(Color Override=(R=1,G=1,B=1,A=1))
   Get Distances Between = RigVMCollapseNode(ExecuteContext=(), Initial=False, Items=())
   Hand Align Solve = RigVMCollapseNode(ExecuteContext=())
   Match FK = RigVMCollapseNode(ExecuteContext=())
   Match IK = RigVMCollapseNode(ExecuteContext=())
   Offset Rotation On Transforms = RigVMCollapseNode(ExecuteContext=(), Rotation Offset=(), Transforms Input=())
   Select Control Shape = RigVMCollapseNode(ExecuteContext=(), False Name=Sphere_Solid, Item=(Type=Control,Name="None"), True Name=Box_Solid, Value=False)
   Soft IK = RigVMCollapseNode(
    Bone A=(Type=Bone,Name="None")
    Bone B=(Type=Bone,Name="None")
    Bone C=(Type=Bone,Name="None")
    Debug=False
    Effector=(Rotation=(X=0,Y=0,Z=0,W=1),Translation=(X=0,Y=0,Z=0),Scale3D=(X=1,Y=1,Z=1))
    EnableStretch=False
    ExecuteContext=()
    Lower Segment Scale=1
    PoleVector=(X=0,Y=0,Z=1)
    PoleVectorKind=Direction
    PoleVectorSpace=(Type=Bone,Name="None")
    PrimaryAxis=(X=1,Y=0,Z=0)
    SecondaryAxis=(X=0,Y=1,Z=0)
    SecondaryAxisWeight=1
    Softness=0
    Softness Distance=0.9
    StretchMaximumRatio=2
    StretchStartRatio=1
    Upper Segment Scale=1
    Weight=1
   )
  ]}
 RigVMFunction_Sequence_1_ContainedGraph {nodes [
   n1 = RigVMFunctionEntryNode()
   RigVMFunction_Sequence = &RigVMFunction_Sequence(ExecuteContext=@n1.ExecuteContext)
   n2 = &RigVMFunction_Sequence(ExecuteContext=@RigVMFunction_Sequence.B)
   n3 = &RigVMFunction_Sequence(ExecuteContext=@n2.B)
   n4 = &RigVMFunction_Sequence(ExecuteContext=@n3.B)
   n5 = &RigVMFunction_Sequence(ExecuteContext=@n4.B)
   n6 = &RigVMFunction_Sequence(ExecuteContext=@n5.B)
   n7 = &RigVMFunction_Sequence(ExecuteContext=@n6.B)
   RigVMFunctionReturnNode(A=@RigVMFunction_Sequence.A, B=@n2.A, C=@n3.A, D=@n4.A, E=@n5.A, F=@n6.A, G=@n7.A, H=@n7.B)
  ]}
 RigVMFunction_Sequence_1_ContainedGraph {nodes [
   n1 = RigVMFunctionEntryNode()
   RigVMFunction_Sequence = &RigVMFunction_Sequence(ExecuteContext=@n1.ExecuteContext)
   n2 = &RigVMFunction_Sequence(ExecuteContext=@RigVMFunction_Sequence.B)
   n3 = &RigVMFunction_Sequence(ExecuteContext=@n2.B)
   n4 = &RigVMFunction_Sequence(ExecuteContext=@n3.B)
   n5 = &RigVMFunction_Sequence(ExecuteContext=@n4.B)
   n6 = &RigVMFunction_Sequence(ExecuteContext=@n5.B)
   n7 = &RigVMFunction_Sequence(ExecuteContext=@n6.B)
   RigVMFunctionReturnNode(A=@RigVMFunction_Sequence.A, B=@n2.A, C=@n3.A, D=@n4.A, E=@n5.A, F=@n6.A, G=@n7.A, H=@n7.B)
  ]}
 RigVMFunction_Sequence_1_ContainedGraph {nodes [
   n1 = RigVMFunctionEntryNode()
   RigVMFunction_Sequence = &RigVMFunction_Sequence(ExecuteContext=@n1.ExecuteContext)
   n2 = &RigVMFunction_Sequence(ExecuteContext=@RigVMFunction_Sequence.B)
   n3 = &RigVMFunction_Sequence(ExecuteContext=@n2.B)
   n4 = &RigVMFunction_Sequence(ExecuteContext=@n3.B)
   RigVMFunctionReturnNode(A=@RigVMFunction_Sequence.A, B=@n2.A, C=@n3.A, D=@n4.A, E=@n4.B)
  ]}
 RigVMFunction_Sequence_1_ContainedGraph {nodes [
   n1 = RigVMFunctionEntryNode()
   RigVMFunction_Sequence = &RigVMFunction_Sequence(ExecuteContext=@n1.ExecuteContext)
   n2 = &RigVMFunction_Sequence(ExecuteContext=@RigVMFunction_Sequence.B)
   n3 = &RigVMFunction_Sequence(ExecuteContext=@n2.B)
   n4 = &RigVMFunction_Sequence(ExecuteContext=@n3.B)
   RigVMFunctionReturnNode(A=@RigVMFunction_Sequence.A, B=@n2.A, C=@n3.A, D=@n4.A, E=@n4.B)
  ]}
 RigVMFunction_Sequence_1_ContainedGraph {nodes [
   n1 = RigVMFunctionEntryNode()
   RigVMFunction_Sequence = &RigVMFunction_Sequence(ExecuteContext=@n1.ExecuteContext)
   n2 = &RigVMFunction_Sequence(ExecuteContext=@RigVMFunction_Sequence.B)
   RigVMFunctionReturnNode(A=@RigVMFunction_Sequence.A, B=@n2.A, C=@n2.B)
  ]}
 RigVMFunction_Sequence_3_ContainedGraph {nodes [
   n1 = RigVMFunctionEntryNode()
   n2 = &RigVMFunction_Sequence(ExecuteContext=@n1.ExecuteContext)
   n3 = &RigVMFunction_Sequence(ExecuteContext=@n2.B)
   RigVMFunctionReturnNode(A=@n2.A, B=@n3.A, C=@n3.B)
  ]}
 RigVMModel Backwards Solve Graph {nodes [
   n1 = /Script/ControlRig.RigUnit_InverseExecution() {is=struct RigVMUnitNode}
   RigVMFunction_Sequence = RigVMAggregateNode(ExecuteContext=@n1.ExecutePin)
   For_Each = RigVMDispatchNode(ExecuteContext=@RigVMFunction_Sequence.A, Array=RigVMVariableNode() {variable=Bones}.Value) {template="DISPATCH_RigVMDispatch_ArrayIterator(in Array,out Element,out Index,out Count,out Ratio)"}
   n2 = RigVMDispatchNode(Index=@For_Each.Index, Array=RigVMVariableNode() {variable=Virtual Bones}.Value) {template="DISPATCH_RigVMDispatch_ArrayGetAtIndex(in Array,in Index,out Element)"}
   > &RigUnit_SetTransform(
    Weight=1
    bInitial=False
    Item=@n2.Element
    Value=/Script/ControlRig.RigUnit_ProjectTransformToNewParent(
     bChildInitial=True
     bNewParentInitial=False
     bOldParentInitial=True
     Child=@n2.Element
     NewParent=@For_Each.Element
     OldParent=@For_Each.Element
    ) {is=struct RigVMUnitNode}.Transform
   )
   Match FK = RigVMFunctionReferenceNode(ExecuteContext=@RigVMFunction_Sequence.C)
   SetAnimationChannel = /Script/ControlRig.RigUnit_SetFloatAnimationChannel(
    Channel=PV Twist Follow
    Value=1
    bInitial=False
    ExecutePin=@RigVMFunction_Sequence.D
    Control=RigVMVariableNode() {variable=IK Control}.Value.Name
   ) {is=struct RigVMUnitNode, template="SetAnimationChannel::Execute(in Value,in Control,in Channel,in bInitial)"}
   > Match IK = RigVMFunctionReferenceNode()
   Compute Mid Null Transform = RigVMFunctionReferenceNode(Items=RigVMVariableNode() {variable=Virtual Bones}.Value)
   > &RigUnit_SetTransform(Weight=1, bInitial=False, Value=@"Compute Mid Null Transform".Struct, Item=RigVMVariableNode() {variable=Mid Null}.Value)
   n3 = RigVMVariableNode() {variable=Virtual Bones}
   > &RigUnit_SetTransform(
    Weight=1
    bInitial=false
    bPropagateToChildren=true
    Value=&RigUnit_GetTransform(
     bInitial=False
     Item=RigVMDispatchNode(Index=1, Array=@n3.Value) {template="DISPATCH_RigVMDispatch_ArrayGetAtIndex(in Array,in Index,out Element)"}.Element
    ).Transform
    Item=RigVMVariableNode() {variable=Mid Control}.Value
   )
   Get Distances Between_0 = RigVMFunctionReferenceNode(Initial=False, ExecuteContext=@RigVMFunction_Sequence.E, Items=RigVMVariableNode() {variable=Virtual Bones}.Value)
   > "Calculate And Set Segment Scale Channel  Values" = RigVMFunctionReferenceNode(
    Segment Scale Channels=((Type=None,Name="None"),(Type=None,Name="None"))
    Current Lengths=@"Get Distances Between_0".Distances
    Segment Scale Channels.0=RigVMVariableNode() {variable=Upper Segment Scale Channel}.Value
    Segment Scale Channels.1=RigVMVariableNode() {variable=Lower Segment Scale Channel}.Value
    Default Lengths=RigVMVariableNode() {variable=Default Lengths}.Value
   )
   n4 = RigVMVariableNode() {variable=Bones}
   &RigUnit_SetTransform(
    Weight=1
    bInitial=False
    Value=&RigUnit_GetTransform(
     bInitial=False
     Item=RigVMDispatchNode(Index=0, Array=@n4.Value) {template="DISPATCH_RigVMDispatch_ArrayGetAtIndex(in Array,in Index,out Element)"}.Element
    ).Transform
    Item=RigVMVariableNode() {variable=Start Twist PV}.Value
    ExecutePin=@RigVMFunction_Sequence.B
   )
   > For_Each_4 = RigVMDispatchNode(Array=RigVMVariableNode() {variable=FK Spaces}.Value) {template="DISPATCH_RigVMDispatch_ArrayIterator(in Array,out Element,out Index,out Count,out Ratio)"}
   n5 = RigVMDispatchNode(Index=@For_Each_4.Index, Array=RigVMVariableNode() {variable=FK Nulls}.Value) {template="DISPATCH_RigVMDispatch_ArrayGetAtIndex(in Array,in Index,out Element)"}
   Multiply = &RigVMFunction_MathTransformMul(
    B=&RigUnit_GetTransform(bInitial=False, Item=@For_Each_4.Element).Transform
    A=&RigVMFunction_MathTransformMakeRelative(
     Parent=&RigUnit_GetTransform(bInitial=true, Item=@For_Each_4.Element).Transform
     Global=&RigUnit_GetTransform(bInitial=true, Item=@n5.Element).Transform
    ).Local
   ) {template="Multiply::Execute(in A,in B,out Result)"}
   > Set Transform_5 = &RigUnit_SetRotation(Space=GlobalSpace, Weight=1, bInitial=False, Item=@n5.Element, Value=@Multiply.Result.Rotation)
  ]}
 RigVMModel Connector Event Graph {nodes [
   n1 = /Script/ControlRig.RigUnit_ConnectorExecution() {is=struct RigVMUnitNode}
   RigVMFunction_ControlFlowBranch = &RigVMFunction_ControlFlowBranch(
    ExecuteContext=@n1.ExecuteContext
    Condition=RigVMDispatchNode(B=(Type=Connector,Name="Parent"), A=/Script/ControlRig.RigUnit_GetCandidates() {is=struct RigVMUnitNode}.Connector) {template="DISPATCH_RigVMDispatch_CoreEquals(in A,in B,out Result)"}.Result
   )
   /Script/ControlRig.RigUnit_SetDefaultMatch(
    Default=&RigUnit_HierarchyGetParent(
     bDefaultParent=True
     Child=/Script/ControlRig.RigUnit_ResolveConnector(Connector=(Type=Connector,Name="Start"), SkipSocket=False) {is=struct RigVMUnitNode}.Result
    ).Parent
    ExecutePin=@RigVMFunction_ControlFlowBranch.True
   ) {is=struct RigVMUnitNode, template="Set Default Match::Execute(in Default)"}
  ]}
 RigVMModel Construction Graph {nodes [
   n1 = RigVMVariableNode() {variable=IK Control}
   n2 = &RigUnit_HierarchyAddNull(
    ExecutePin=()
    Name=IK Twist Null
    Space=GlobalSpace
    Transform=&RigUnit_GetTransform(bInitial=true, Item=@n1.Value).Transform
    Parent=&RigUnit_HierarchyGetParent(bDefaultParent=true, Child=@n1.Value).Parent
   )
   > RigVMVariableNode(Value=@n2.Item) {variable=IK Twist Null}
   n3 = &RigUnit_PrepareForExecution()
   RigVMVariableNode(
    Value=((Type=Connector,Name="Start"),(Type=Connector,Name="Mid"),(Type=Connector,Name="End"))
    ExecuteContext=@n3.ExecutePin
    Value.0=&RigUnit_ResolveConnector(Connector=(Type=Connector,Name="Start"), SkipSocket=False).Result
    Value.1=&RigUnit_ResolveConnector(Connector=(Type=Connector,Name="Mid"), SkipSocket=False).Result
    Value.2=&RigUnit_ResolveConnector(Connector=(Type=Connector,Name="End"), SkipSocket=False).Result
   ) {variable=Bones}
   > RigVMVariableNode(Value=/Script/ControlRig.RigUnit_ResolveArrayConnector(Connector=(Type=Connector,Name="IK Spaces"), SkipSocket=False) {is=struct RigVMUnitNode}.Result) {variable=IK Spaces}
   > RigVMVariableNode(Value=&RigUnit_ResolveArrayConnector(Connector=(Type=Connector,Name="FK Spaces"), SkipSocket=False).Result) {variable=FK Spaces}
   DISPATCH_RigVMDispatch_ArrayGetAtIndex = RigVMDispatchNode(Index=0, Array=RigVMVariableNode() {variable=Bones}.Value) {template="DISPATCH_RigVMDispatch_ArrayGetAtIndex(in Array,in Index,out Element)"}
   > RigVMVariableNode(Value=&RigUnit_HierarchyGetParent(bDefaultParent=True, Child=@DISPATCH_RigVMDispatch_ArrayGetAtIndex.Element).Parent) {variable=Parent}
   > Get Distances Between = RigVMFunctionReferenceNode(Initial=true, Items=RigVMVariableNode() {variable=Bones}.Value)
   > RigVMVariableNode(Value=@"Get Distances Between".Distances) {variable=Default Lengths}
   Get Control Color From Metadata = RigVMFunctionReferenceNode(Color Override=RigVMVariableNode() {variable=Color}.Value)
   > RigVMVariableNode(Value=@"Get Control Color From Metadata".Result) {variable=Color}
   > DISPATCH_RigVMDispatch_ArrayReset = RigVMDispatchNode(Array=RigVMVariableNode() {variable=FK Controls}.Value) {template="DISPATCH_RigVMDispatch_ArrayReset(io Array)"}
   > RigVMDispatchNode(Array=RigVMVariableNode() {variable=FK Gimbal Controls}.Value) {template="DISPATCH_RigVMDispatch_ArrayReset(io Array)"}
   > RigVMDispatchNode(Array=RigVMVariableNode() {variable=FK Drivers}.Value) {template="DISPATCH_RigVMDispatch_ArrayReset(io Array)"}
   > RigVMFunction_Sequence = RigVMAggregateNode()
   n4 = RigVMVariableNode() {variable=Bones}
   n5 = &RigUnit_HierarchyAddNull(
    Name=Default FK Space
    Space=GlobalSpace
    Transform=&RigUnit_GetTransform(
     bInitial=true
     Item=RigVMDispatchNode(Index=0, Array=@n4.Value) {template="DISPATCH_RigVMDispatch_ArrayGetAtIndex(in Array,in Index,out Element)"}.Element
    ).Transform
    Parent=RigVMVariableNode() {variable=Parent}.Value
    ExecutePin=@RigVMFunction_Sequence.B
   )
   > RigVMVariableNode(Value=@n5.Item) {variable=FK Parent}
   > For_Each = RigVMDispatchNode(Array=RigVMVariableNode() {variable=Virtual Bones}.Value) {template="DISPATCH_RigVMDispatch_ArrayIterator(in Array,out Element,out Index,out Count,out Ratio)"}
   > Get Item Name v01 = RigVMFunctionReferenceNode(
    Prefix=FK
    Remove Side Suffix=true
    Standard Numerical Names=true
    Name=RigVMDispatchNode(Index=@For_Each.Index, Array=RigVMVariableNode() {variable=Bones}.Value) {template="DISPATCH_RigVMDispatch_ArrayGetAtIndex(in Array,in Index,out Element)"}.Element.Name
    Count=@For_Each.Count
    Index=@For_Each.Index
   )
   n6 = &RigVMFunction_ControlFlowBranch(
    Condition=RigVMDispatchNode(B=0, A=@For_Each.Index) {template="DISPATCH_RigVMDispatch_CoreEquals(in A,in B,out Result)"}.Result
    ExecuteContext=@"Get Item Name v01".ExecuteContext
   )
   n7 = RigVMVariableNode() {variable=FK Control Shape}
   Control Color Override From Metadata_1 = RigVMFunctionReferenceNode(Color Override=@n7.Value.Color)
   n8 = RigVMVariableNode() {variable=FK Control Shape}
   n9 = RigVMDispatchNode(
    True=&RigVMFunction_MathTransformMul(B=(Rotation=(X=0,Y=0,Z=0,W=1),Translation=(X=0,Y=0,Z=0),Scale3D=(X=-1,Y=1,Z=1)), A=@n8.Value.Transform) {template="Multiply::Execute(in A,in B,out Result)"}.Result
    Condition=RigVMVariableNode() {variable=Negative Side}.Value
    False=@n8.Value.Transform
   ) {template="DISPATCH_RigVMDispatch_If(in Condition,in True,in False,out Result)"}
   RigVMFunction_AnimEvalRichCurve = &RigVMFunction_AnimEvalRichCurve(SourceMinimum=0, TargetMinimum=0, Value=@For_Each.Ratio, Curve=RigVMVariableNode() {variable=FK Control Scale Factor Profile}.Value)
   Scale = &RigVMFunction_MathVectorScale(Value=@n9.Result.Scale3D, Factor=@RigVMFunction_AnimEvalRichCurve.Result) {template="Scale::Execute(in Value,in Factor,out Result)"}
   Make = RigVMDispatchNode(
    Elements=(bVisible=True,Name="Default",Color=(R=1,G=0,B=0,A=1),Transform=(Rotation=(X=0,Y=0,Z=0,W=1),Translation=(X=0,Y=0,Z=0),Scale3D=(X=1,Y=1,Z=1)))
    Elements.Color=@"Control Color Override From Metadata_1".Result
    Elements.Transform.Translation=@n9.Result.Translation
    Elements.Transform.Rotation=&RigVMFunction_MathQuaternionMul(A=@n9.Result.Rotation, B=RigVMVariableNode() {variable=FK Rotation Offset}.Value) {template="Multiply::Execute(in A,in B,out Result)"}.Result
    Elements.Transform.Scale3D=@Scale.Result
    Elements.Name=@n7.Value.Name
    Elements.bVisible=@n7.Value.bVisible
   ) {template="DISPATCH_RigVMDispatch_MakeStruct(in Elements,out Struct)"}
   Make_Array_1 = RigVMDispatchNode(
    Values=("None","None","None")
    Values.0=RigVMVariableNode() {variable=FK Start Display Name}.Value
    Values.1=RigVMVariableNode() {variable=FK Mid Display Name}.Value
    Values.2=RigVMVariableNode() {variable=FK End Display Name}.Value
   ) {template="DISPATCH_RigVMDispatch_ArrayMake(in Values,out Array)"}
   n10 = RigVMDispatchNode(Index=@For_Each.Index, Array=@Make_Array_1.Array) {template="DISPATCH_RigVMDispatch_ArrayGetAtIndex(in Array,in Index,out Element)"}
   n11 = &RigUnit_GetTransform(bInitial=true, Item=@For_Each.Element)
   n12 = RigVMVariableNode() {variable=FK Rotation Offset}
   SpawnControl = /Script/ControlRig.RigUnit_HierarchyAddControlTransform(
    InitialValue=(Rotation=(X=0,Y=0,Z=0,W=1),Translation=(X=0,Y=0,Z=0),Scale3D=(X=1,Y=1,Z=1))
    OffsetSpace=GlobalSpace
    Settings="(InitialSpace=LocalSpace,bUsePreferredRotationOrder=true,PreferredRotationOrder=YZX,Limits=(LimitTranslationX=(bMinimum=False,bMaximum=False),LimitTranslationY=(bMinimum=False,bMaximum=False),LimitTranslationZ=(bMinimum=False,bMaximum=False),LimitPitch=(bM ...(truncated)"
    Settings.Shape=@Make.Struct
    ExecutePin=@n6.Completed
    Parent=RigVMVariableNode() {variable=FK Parent}.Value
    Settings.PreferredRotationOrder=RigVMVariableNode() {variable=Rotation Order}.Value
    Settings.DisplayName=@n10.Element
    Name=@"Get Item Name v01".Result
    OffsetTransform=&RigVMFunction_MathTransformMake(
     Scale=@n11.Transform.Scale3D
     Translation=@n11.Transform.Translation
     Rotation=&RigVMFunction_MathQuaternionMul(A=@n11.Transform.Rotation, B=@n12.Value) {template="Multiply::Execute(in A,in B,out Result)"}.Result
    ).Result
   ) {
    is=struct RigVMUnitNode
    template="SpawnControl::Execute(in InitialValue,in Settings,in OffsetTransform,in OffsetSpace,in Parent,in Name,out Item)"
   }
   > RigVMVariableNode(Value=@SpawnControl.Item) {variable=FK Parent}
   n13 = RigVMVariableNode() {variable=FK Parent}
   > RigVMDispatchNode(Array=RigVMVariableNode() {variable=FK Controls}.Value, Element=@n13.Value) {template="DISPATCH_RigVMDispatch_ArrayAdd(io Array,in Element,out Index)"}
   > Set Mirror Behavior = RigVMFunctionReferenceNode(Value=RigVMVariableNode() {variable=FK Mirror Behavior}.Value, Item=@n13.Value)
   Not_Equals_1 = RigVMDispatchNode(B=1, A=@For_Each.Index) {template="DISPATCH_RigVMDispatch_CoreNotEquals(in A,in B,out Result)"}
   n14 = &RigVMFunction_ControlFlowBranch(Condition=@Not_Equals_1.Result, ExecuteContext=@"Set Mirror Behavior".ExecuteContext)
   RigVMDispatchNode(ExecuteContext=@n14.False, Array=RigVMVariableNode() {variable=FK Drivers}.Value, Element=@n13.Value) {template="DISPATCH_RigVMDispatch_ArrayAdd(io Array,in Element,out Index)"}
   Create Gimbal Control = RigVMFunctionReferenceNode(
    Shape Settings Override=(bVisible=True,Name="None",Color=(R=1,G=0,B=0,A=1),Transform=(Rotation=(X=0,Y=0,Z=0,W=1),Translation=(X=0,Y=0,Z=0),Scale3D=(X=1,Y=1,Z=1)))
    ExecuteContext=@n14.True
    Rotation Order=RigVMVariableNode() {variable=Rotation Order}.Value
    Display Name=@n10.Element
    Source Control=@n13.Value
   )
   > RigVMVariableNode(Value=@"Create Gimbal Control".Gimbal Control) {variable=FK Parent}
   > Set Mirror Behavior_2 = RigVMFunctionReferenceNode(Value=RigVMVariableNode() {variable=FK Mirror Behavior}.Value, Item=@"Create Gimbal Control".Gimbal Control)
   > RigVMDispatchNode(Array=RigVMVariableNode() {variable=FK Gimbal Controls}.Value, Element=@"Create Gimbal Control".Gimbal Control) {template="DISPATCH_RigVMDispatch_ArrayAdd(io Array,in Element,out Index)"}
   > RigVMDispatchNode(Array=RigVMVariableNode() {variable=FK Drivers}.Value, Element=@"Create Gimbal Control".Gimbal Control) {template="DISPATCH_RigVMDispatch_ArrayAdd(io Array,in Element,out Index)"}
   Make_Array_2 = RigVMDispatchNode(Values=("None","Upper Segment Scale FK","Lower Segment Scale FK")) {template="DISPATCH_RigVMDispatch_ArrayMake(in Values,out Array)"}
   n15 = RigVMVariableNode() {variable=FK Parent}
   RigVMFunction_MathTransformMake = &RigVMFunction_MathTransformMake(
    Scale=@n11.Transform.Scale3D
    Translation=@n11.Transform.Translation
    Rotation=&RigUnit_GetTransform(bInitial=true, Item=@n15.Value).Transform.Rotation
   )
   n16 = &RigUnit_HierarchyAddNull(
    Space=GlobalSpace
    Name=RigVMDispatchNode(Array=@Make_Array_2.Array, Index=@For_Each.Index) {template="DISPATCH_RigVMDispatch_ArrayGetAtIndex(in Array,in Index,out Element)"}.Element
    ExecutePin=@n6.False
    Transform=@RigVMFunction_MathTransformMake.Result
    Parent=@n15.Value
   )
   > RigVMVariableNode(Value=@n16.Item) {variable=FK Parent}
   n17 = RigVMDispatchNode(
    Array=RigVMDispatchNode(Array=RigVMVariableNode() {variable=FK Controls}.Value) {template="DISPATCH_RigVMDispatch_ArrayClone(in Array,out Clone)"}.Clone
    Other=RigVMVariableNode() {variable=FK Gimbal Controls}.Value
    ExecuteContext=@RigVMFunction_Sequence.H
   ) {template="DISPATCH_RigVMDispatch_ArrayAppend(io Array,in Other)"}
   Make_Array = RigVMDispatchNode(
    Values=((Type=None,Name="None"),((),Type=None,Name="None"),((),Type=None,Name="None"),((),Type=None,Name="None"),((),Type=None,Name="None"),((),Type=None,Name="None"))
    Values.3=RigVMVariableNode() {variable=IK Rotation Control}.Value
    Values.4=RigVMVariableNode() {variable=IK Gimbal Control}.Value
    Values.5=RigVMVariableNode() {variable=IK Base Control}.Value
    Values.0=RigVMVariableNode() {variable=IK Control}.Value
    Values.1=RigVMVariableNode() {variable=PV Control}.Value
    Values.2=RigVMVariableNode() {variable=Mid Control}.Value
   ) {template="DISPATCH_RigVMDispatch_ArrayMake(in Values,out Array)"}
   > n18 = RigVMDispatchNode(Array=@n17.Array, Other=@Make_Array.Array) {template="DISPATCH_RigVMDispatch_ArrayUnion(io Array,in Other)"}
   n19 = RigVMVariableNode() {variable=Bones}
   > Get Chain Length = RigVMFunctionReferenceNode(Initial=true, Items=@n19.Value)
   Get_Module_Metadata_1 = RigVMDispatchNode(Default=1, Name=Global Control Scale, NameSpace=Root) {template="DISPATCH_RigDispatch_GetModuleMetadata(in Name,in NameSpace,in Default,out Value,out Found)"}
   Num = RigVMDispatchNode(Array=@n19.Value) {template="DISPATCH_RigVMDispatch_ArrayGetNum(in Array,out Num)"}
   > Set Control Scale = RigVMFunctionReferenceNode(
    Global Scale=@Get_Module_Metadata_1.Value
    Auto Scale Factor=RigVMDispatchNode(
     False=1
     True=&RigVMFunction_MathDoubleDiv(B=55, A=@"Get Chain Length".Length).Result
     Condition=/Script/RigVM.RigVMFunction_MathIntGreater(B=1, A=@Num.Num) {is=struct RigVMUnitNode, template="Greater::Execute(in A,in B,out Result)"}.Result
    ) {template="DISPATCH_RigVMDispatch_If(in Condition,in True,in False,out Result)"}.Result
    Scale=RigVMVariableNode() {variable=Control Scale}.Value
    Controls=@n18.Array
   )
   > Switch Control Visibility = RigVMFunctionReferenceNode(
    On Controls=((Type=None,Name="None"),(Type=None,Name="None"))
    Switch=RigVMVariableNode() {variable=IK FK Manipulation}.Value
    Off Controls=RigVMVariableNode() {variable=FK Controls}.Value
    On Controls.0=RigVMVariableNode() {variable=IK Control}.Value
    On Controls.1=RigVMVariableNode() {variable=PV Control}.Value
   )
   > n20 = /Script/ControlRig.RigUnit_HierarchyAddAnimationChannelBool(
    InitialValue=false
    LimitsEnabled=()
    MaximumValue=True
    MinimumValue=false
    Name=Sec Controls Vis
    Parent=RigVMVariableNode() {variable=IK Control}.Value
   ) {
    is=struct RigVMUnitNode
    template="SpawnAnimationChannel::Execute(in InitialValue,in MinimumValue,in MaximumValue,in LimitsEnabled,in Parent,in Name,out Item)"
   }
   > /Script/ControlRig.RigUnit_SetChannelHosts(
    Hosts=((Type=None,Name="None"))
    Hosts.0=RigVMDispatchNode(Index=-1, Array=RigVMVariableNode() {variable=FK Controls}.Value) {template="DISPATCH_RigVMDispatch_ArrayGetAtIndex(in Array,in Index,out Element)"}.Element
    Channel=@n20.Item
   ) {is=struct RigVMUnitNode}
   > RigVMVariableNode(Value=@n20.Item) {variable=Sec Controls Vis Channel}
   > n21 = &RigUnit_HierarchyAddAnimationChannelBool(
    InitialValue=False
    LimitsEnabled=()
    MaximumValue=True
    MinimumValue=False
    Name=IK Base Control Vis
    Parent=RigVMVariableNode() {variable=IK Base Control}.Value
   ) {template="SpawnAnimationChannel::Execute(in InitialValue,in MinimumValue,in MaximumValue,in LimitsEnabled,in Parent,in Name,out Item)"}
   > &RigUnit_SetChannelHosts(
    Hosts=((Type=None,Name="None"),(Type=None,Name="None"))
    Hosts.0=RigVMVariableNode() {variable=IK Control}.Value
    Hosts.1=RigVMVariableNode() {variable=IK Gimbal Control}.Value
    Channel=@n21.Item
   )
   > RigVMVariableNode(Value=@n21.Item) {variable=IK Base Control Vis Channel}
   RigVMFunction_ControlFlowBranch = &RigVMFunction_ControlFlowBranch(Condition=RigVMVariableNode() {variable=IK FK Manipulation}.Value, ExecuteContext=@RigVMFunction_Sequence.F)
   n22 = RigVMDispatchNode(ExecuteContext=@RigVMFunction_ControlFlowBranch.True, Array=RigVMVariableNode() {variable=FK Controls}.Value) {template="DISPATCH_RigVMDispatch_ArrayIterator(in Array,out Element,out Index,out Count,out Ratio)"}
   RigUnit_ItemToName = /Script/ControlRig.RigUnit_ItemToName(Value=@n22.Element) {is=struct RigVMUnitNode, template="Cast::Execute(in Value,out Result)"}
   > SetControlDrivenList = /Script/ControlRig.RigUnit_SetControlDrivenList(
    Driven=((Type=None,Name="None"),(Type=None,Name="None"))
    Control=@RigUnit_ItemToName.Result
    Driven.0=RigVMVariableNode() {variable=IK Control}.Value
    Driven.1=RigVMVariableNode() {variable=PV Control}.Value
   ) {is=struct RigVMUnitNode, template="SetControlDrivenList::Execute(in Control,in Driven)"}
   n23 = RigVMDispatchNode(
    Array=((Type=None,Name="None"),(Type=None,Name="None"))
    Array.0=RigVMVariableNode() {variable=IK Control}.Value
    Array.1=RigVMVariableNode() {variable=PV Control}.Value
    ExecuteContext=@n22.Completed
   ) {template="DISPATCH_RigVMDispatch_ArrayIterator(in Array,out Element,out Index,out Count,out Ratio)"}
   > &RigUnit_SetControlDrivenList(
    Control=&RigUnit_ItemToName(Value=@n23.Element) {template="Cast::Execute(in Value,out Result)"}.Result
    Driven=RigVMVariableNode() {variable=FK Controls}.Value
   ) {template="SetControlDrivenList::Execute(in Control,in Driven)"}
   Compute Mid Null Transform = RigVMFunctionReferenceNode(Items=RigVMVariableNode() {variable=Virtual Bones}.Value)
   n24 = &RigUnit_HierarchyAddNull(
    Name=Mid
    Parent=(Type=Bone,Name="None")
    Space=GlobalSpace
    Transform=@"Compute Mid Null Transform".Struct
    ExecutePin=@RigVMFunction_Sequence.G
   )
   > RigVMVariableNode(Value=@n24.Item) {variable=Mid Null}
   Control Color Override From Metadata_3 = RigVMFunctionReferenceNode(Color Override=RigVMVariableNode() {variable=IK Control Shape}.Value.Color)
   n25 = RigVMVariableNode() {variable=Negative Side}
   > n26 = &RigUnit_HierarchyAddControlTransform(
    InitialValue=(Rotation=(X=0,Y=0,Z=0,W=1),Translation=(X=0,Y=0,Z=0),Scale3D=(X=1,Y=1,Z=1))
    Name=Mid
    OffsetSpace=LocalSpace
    OffsetTransform=(Rotation=(X=0,Y=0,Z=0,W=1),Translation=(X=0,Y=0,Z=0),Scale3D=(X=1,Y=1,Z=1))
    Settings="(InitialSpace=LocalSpace,bUsePreferredRotationOrder=False,PreferredRotationOrder=YZX,Limits=(LimitTranslationX=(bMinimum=False,bMaximum=False),LimitTranslationY=(bMinimum=False,bMaximum=False),LimitTranslationZ=(bMinimum=False,bMaximum=False),LimitPitch=(b ...(truncated)"
    Settings.Shape.Color=@"Control Color Override From Metadata_3".Result
    Settings.Shape.Transform.Rotation=RigVMDispatchNode(False=(X=0,Y=-0.7071,Z=0,W=0.7071), True=(X=0.7071,Y=0,Z=-0.7071,W=0), Condition=@n25.Value) {template="DISPATCH_RigVMDispatch_If(in Condition,in True,in False,out Result)"}.Result
    Parent=@n24.Item
    Settings.Shape.Transform.Translation=RigVMDispatchNode(False=(X=0,Y=-6,Z=0), True=(X=0,Y=6,Z=0), Condition=@n25.Value) {template="DISPATCH_RigVMDispatch_If(in Condition,in True,in False,out Result)"}.Result
    Settings.DisplayName=RigVMVariableNode() {variable=IK Mid Display Name}.Value
   ) {template="SpawnControl::Execute(in InitialValue,in Settings,in OffsetTransform,in OffsetSpace,in Parent,in Name,out Item)"}
   > RigVMVariableNode(Value=@n26.Item) {variable=Mid Control}
   > Set Mirror Behavior_1 = RigVMFunctionReferenceNode(Item=@n26.Item, Value=RigVMVariableNode() {variable=FK Mirror Behavior}.Value)
   n27 = RigVMVariableNode() {variable=IK Control}
   > n28 = &RigUnit_HierarchyAddAnimationChannelBool(
    LimitsEnabled=()
    MaximumValue=True
    MinimumValue=false
    Name=Ik Fk Switch
    InitialValue=RigVMVariableNode() {variable=Default IK}.Value
    Parent=@n27.Value
   ) {template="SpawnAnimationChannel::Execute(in InitialValue,in MinimumValue,in MaximumValue,in LimitsEnabled,in Parent,in Name,out Item)"}
   > RigVMVariableNode(Value=@n28.Item) {variable=Ik Solve Channel}
   Clone = RigVMDispatchNode(Array=RigVMVariableNode() {variable=FK Controls}.Value) {template="DISPATCH_RigVMDispatch_ArrayClone(in Array,out Clone)"}
   > n29 = RigVMDispatchNode(Array=@Clone.Clone, Other=RigVMVariableNode() {variable=FK Gimbal Controls}.Value) {template="DISPATCH_RigVMDispatch_ArrayAppend(io Array,in Other)"}
   > Add = RigVMDispatchNode(Array=@n29.Array, Element=RigVMVariableNode() {variable=Mid Control}.Value) {template="DISPATCH_RigVMDispatch_ArrayAdd(io Array,in Element,out Index)"}
   > n30 = RigVMDispatchNode(Array=@Add.Array, Element=RigVMVariableNode() {variable=PV Control}.Value) {template="DISPATCH_RigVMDispatch_ArrayAdd(io Array,in Element,out Index)"}
   > n31 = RigVMDispatchNode(Array=@n30.Array, Element=RigVMVariableNode() {variable=IK Gimbal Control}.Value) {template="DISPATCH_RigVMDispatch_ArrayAdd(io Array,in Element,out Index)"}
   > n32 = RigVMDispatchNode(Array=@n31.Array, Element=RigVMVariableNode() {variable=IK Base Control}.Value) {template="DISPATCH_RigVMDispatch_ArrayAdd(io Array,in Element,out Index)"}
   > n33 = &RigUnit_SetChannelHosts(Hosts=@n32.Array, Channel=@n28.Item)
   n34 = &RigVMFunction_ControlFlowBranch(ExecuteContext=@n33.ExecutePin, Condition=RigVMVariableNode() {variable=Hand Align}.Value)
   &RigUnit_HierarchyAddAnimationChannelBool(InitialValue=true, LimitsEnabled=(), MaximumValue=True, MinimumValue=False, Name=Stretch, ExecutePin=@n34.Completed, Parent=@n27.Value) {template="SpawnAnimationChannel::Execute(in InitialValue,in MinimumValue,in MaximumValue,in LimitsEnabled,in Parent,in Name,out Item)"}
   > /Script/ControlRig.RigUnit_HierarchyAddAnimationChannelFloat(InitialValue=0, LimitsEnabled=(Enabled=(bMinimum=True,bMaximum=True)), MaximumValue=1, MinimumValue=0, Name=Softness, Parent=@n27.Value) {
    is=struct RigVMUnitNode
    template="SpawnAnimationChannel::Execute(in InitialValue,in MinimumValue,in MaximumValue,in LimitsEnabled,in Parent,in Name,out Item)"
   }
   > n35 = &RigUnit_HierarchyAddAnimationChannelFloat(
    LimitsEnabled=(Enabled=(bMinimum=True,bMaximum=True))
    MaximumValue=1
    MinimumValue=0
    Name=PV Twist Follow
    InitialValue=RigVMVariableNode() {variable=PV Twist Follow}.Value
    Parent=@n27.Value
   ) {template="SpawnAnimationChannel::Execute(in InitialValue,in MinimumValue,in MaximumValue,in LimitsEnabled,in Parent,in Name,out Item)"}
   > n36 = &RigUnit_SetChannelHosts(Hosts=((Type=None,Name="None")), Channel=@n35.Item, Hosts.0=RigVMVariableNode() {variable=PV Control}.Value)
   n37 = &RigVMFunction_ControlFlowBranch(ExecuteContext=@n36.ExecutePin, Condition=RigVMVariableNode() {variable=Segment Scale Control}.Value)
   n38 = RigVMDispatchNode(
    Array=RigVMDispatchNode(Array=RigVMVariableNode() {variable=FK Controls}.Value) {template="DISPATCH_RigVMDispatch_ArrayClone(in Array,out Clone)"}.Clone
    ExecuteContext=@n37.True
    Element=RigVMVariableNode() {variable=IK Base Control}.Value
   ) {template="DISPATCH_RigVMDispatch_ArrayAdd(io Array,in Element,out Index)"}
   n39 = RigVMVariableNode() {variable=IK Control}
   > SpawnScaleAnimationChannel = /Script/ControlRig.RigUnit_HierarchyAddAnimationChannelScaleFloat(
    InitialValue=1
    LimitsEnabled=(Enabled=(bMinimum=True,bMaximum=false))
    MaximumValue=2
    MinimumValue=0.0001
    Name=Upper Segment Scale
    Parent=@n39.Value
   ) {
    is=struct RigVMUnitNode
    template="SpawnScaleAnimationChannel::Execute(in InitialValue,in MinimumValue,in MaximumValue,in LimitsEnabled,in Parent,in Name,out Item)"
   }
   > RigVMVariableNode(Value=@SpawnScaleAnimationChannel.Item) {variable=Upper Segment Scale Channel}
   > &RigUnit_SetChannelHosts(Channel=@SpawnScaleAnimationChannel.Item, Hosts=@n38.Array)
   > n40 = &RigUnit_HierarchyAddAnimationChannelScaleFloat(
    InitialValue=1
    LimitsEnabled=(Enabled=(bMinimum=True,bMaximum=false))
    MaximumValue=2
    MinimumValue=0.0001
    Name=Lower Segment Scale
    Parent=@n39.Value
   ) {template="SpawnScaleAnimationChannel::Execute(in InitialValue,in MinimumValue,in MaximumValue,in LimitsEnabled,in Parent,in Name,out Item)"}
   > RigVMVariableNode(Value=@n40.Item) {variable=Lower Segment Scale Channel}
   > &RigUnit_SetChannelHosts(Channel=@n40.Item, Hosts=@n38.Array)
   SpawnAnimationChannel = &RigUnit_HierarchyAddAnimationChannelBool(
    LimitsEnabled=()
    MaximumValue=True
    MinimumValue=False
    Name=Hand Align
    InitialValue=RigVMVariableNode() {variable=Hand Align Default}.Value
    Parent=@n27.Value
    ExecutePin=@n34.True
   ) {template="SpawnAnimationChannel::Execute(in InitialValue,in MinimumValue,in MaximumValue,in LimitsEnabled,in Parent,in Name,out Item)"}
   > RigVMVariableNode(Value=@SpawnAnimationChannel.Item) {variable=Hand Align Channel}
   > &RigUnit_SetChannelHosts(Hosts=((Type=None,Name="None")), Channel=@SpawnAnimationChannel.Item, Hosts.0=RigVMVariableNode() {variable=IK Rotation Control}.Value)
   RigVMDispatchNode(Array=RigVMVariableNode() {variable=FK Nulls}.Value, ExecuteContext=@RigVMFunction_Sequence.C) {template="DISPATCH_RigVMDispatch_ArrayReset(io Array)"}
   > For_Each_1 = RigVMDispatchNode(Array=RigVMVariableNode() {variable=FK Spaces}.Value) {template="DISPATCH_RigVMDispatch_ArrayIterator(in Array,out Element,out Index,out Count,out Ratio)"}
   n41 = RigVMVariableNode() {variable=Bones}
   > n42 = &RigUnit_HierarchyAddNull(
    Space=GlobalSpace
    Transform=&RigUnit_GetTransform(
     bInitial=true
     Item=RigVMDispatchNode(Index=0, Array=@n41.Value) {template="DISPATCH_RigVMDispatch_ArrayGetAtIndex(in Array,in Index,out Element)"}.Element
    ).Transform
    Parent=RigVMVariableNode() {variable=Parent}.Value
    Name=/Script/RigVM.RigVMFunction_NameConcat(B=" Orient", A=/Script/ControlRig.RigUnit_GetItemShortName(Item=@For_Each_1.Element) {is=struct RigVMUnitNode}.ShortName) {is=struct RigVMUnitNode, template="Concat::Execute(in A,in B,out Result)"}.Result
   )
   > RigVMDispatchNode(Array=RigVMVariableNode() {variable=FK Nulls}.Value, Element=@n42.Item) {template="DISPATCH_RigVMDispatch_ArrayAdd(io Array,in Element,out Index)"}
   > /Script/ControlRig.RigUnit_AddParent(
    DisplayLabel=None
    Child=RigVMDispatchNode(Index=0, Array=RigVMVariableNode() {variable=FK Controls}.Value) {template="DISPATCH_RigVMDispatch_ArrayGetAtIndex(in Array,in Index,out Element)"}.Element
    Parent=@n42.Item
   ) {is=struct RigVMUnitNode}
   n43 = RigVMVariableNode() {variable=Default FK Space Index}
   Less = /Script/RigVM.RigVMFunction_MathIntLess(
    B=RigVMDispatchNode(Array=RigVMVariableNode() {variable=FK Nulls}.Value) {template="DISPATCH_RigVMDispatch_ArrayGetNum(in Array,out Num)"}.Num
    A=@n43.Value
   ) {is=struct RigVMUnitNode, template="Less::Execute(in A,in B,out Result)"}
   Not_Equals_3 = RigVMDispatchNode(B=-1, A=@n43.Value) {template="DISPATCH_RigVMDispatch_CoreNotEquals(in A,in B,out Result)"}
   RigVMFunction_MathBoolAnd = &RigVMFunction_MathBoolAnd(B=@Less.Result, A=@Not_Equals_3.Result)
   n44 = &RigVMFunction_ControlFlowBranch(Condition=@RigVMFunction_MathBoolAnd.Result, ExecuteContext=@For_Each_1.Completed)
   /Script/ControlRig.RigUnit_SwitchParent(
    Mode=ParentItem
    bMaintainGlobal=True
    Parent=RigVMDispatchNode(Array=RigVMVariableNode() {variable=FK Nulls}.Value, Index=RigVMVariableNode() {variable=Default FK Space Index}.Value) {template="DISPATCH_RigVMDispatch_ArrayGetAtIndex(in Array,in Index,out Element)"}.Element
    Child=RigVMDispatchNode(Index=0, Array=RigVMVariableNode() {variable=FK Controls}.Value) {template="DISPATCH_RigVMDispatch_ArrayGetAtIndex(in Array,in Index,out Element)"}.Element
    ExecutePin=@n44.True
   ) {is=struct RigVMUnitNode}
   n45 = &RigUnit_GetTransform(
    bInitial=true
    Item=RigVMDispatchNode(Index=0, Array=RigVMVariableNode() {variable=Virtual Bones}.Value) {template="DISPATCH_RigVMDispatch_ArrayGetAtIndex(in Array,in Index,out Element)"}.Element
   )
   n46 = &RigUnit_HierarchyAddNull(
    Name=Parent Buffer
    Space=GlobalSpace
    Transform=@n45.Transform
    Parent=RigVMVariableNode() {variable=Parent}.Value
    ExecutePin=@RigVMFunction_Sequence.D
   )
   > RigVMVariableNode(Value=@n46.Item) {variable=Parent Buffer}
   n47 = RigVMVariableNode() {variable=IK Control Shape}
   Control Color Override From Metadata_5 = RigVMFunctionReferenceNode(Color Override=@n47.Value.Color)
   > n48 = &RigUnit_HierarchyAddControlTransform(
    Name=IK Base
    OffsetSpace=GlobalSpace
    Settings="(InitialSpace=GlobalSpace,Shape=(bVisible=True,Name=\"DefaultGizmoLibraryNormalized.Box_Thick\",Color=(R=1,G=0,B=0,A=1),Transform=(Rotation=(X=0,Y=0,Z=0,W=1),Translation=(X=0,Y=0,Z=0),Scale3D=(X=1,Y=1,Z=1))),Proxy=(bIsProxy=False,ShapeVisibility=BasedOnSelec ...(truncated)"
    Settings.Shape.Color=@"Control Color Override From Metadata_5".Result
    InitialValue=@n45.Transform
    OffsetTransform=@n45.Transform
    Parent=@n46.Item
    Settings.Shape.Transform.Scale3D=&RigVMFunction_MathVectorMul(B=(X=1.5,Y=1.5,Z=1.5), A=@n47.Value.Transform.Scale3D) {template="Multiply::Execute(in A,in B,out Result)"}.Result
    Settings.PreferredRotationOrder=RigVMVariableNode() {variable=Rotation Order}.Value
    Settings.Shape.Name=@n47.Value.Name
    Settings.Shape.Transform.Rotation=@n47.Value.Transform.Rotation
    Settings.Shape.Transform.Translation=@n47.Value.Transform.Translation
   ) {template="SpawnControl::Execute(in InitialValue,in Settings,in OffsetTransform,in OffsetSpace,in Parent,in Name,out Item)"}
   > RigVMVariableNode(Value=@n48.Item) {variable=IK Base Control}
   n49 = &RigUnit_GetTransform(
    bInitial=true
    Item=RigVMDispatchNode(Index=-1, Array=RigVMVariableNode() {variable=Virtual Bones}.Value) {template="DISPATCH_RigVMDispatch_ArrayGetAtIndex(in Array,in Index,out Element)"}.Element
   )
   n50 = &RigVMFunction_MathQuaternionMul(A=@n49.Transform.Rotation, B=RigVMVariableNode() {variable=IK Rotation Offset}.Value) {template="Multiply::Execute(in A,in B,out Result)"}
   n51 = RigVMVariableNode() {variable=IK Control Shape}
   > n52 = &RigUnit_HierarchyAddControlTransform(
    InitialValue=(Rotation=(X=0,Y=0,Z=0,W=1),Translation=(X=0,Y=0,Z=0),Scale3D=(X=1,Y=1,Z=1))
    Name=IK
    OffsetSpace=GlobalSpace
    OffsetTransform=(Rotation=(X=0,Y=0,Z=0,W=1),Translation=(X=0,Y=0,Z=0),Scale3D=(X=1,Y=1,Z=1))
    Settings="(InitialSpace=GlobalSpace,Shape=(bVisible=True,Name=\"DefaultGizmoLibraryNormalized.Box_Thick\",Color=(R=1,G=0,B=0,A=1),Transform=(Rotation=(X=0,Y=0,Z=0,W=1),Translation=(X=0,Y=0,Z=0),Scale3D=(X=1,Y=1,Z=1))),Proxy=(bIsProxy=False,ShapeVisibility=BasedOnSelec ...(truncated)"
    Parent=RigVMDispatchNode(Default=(Type=Bone,Name="None"), Name=Local Control, NameSpace=Root) {template="DISPATCH_RigDispatch_GetModuleMetadata(in Name,in NameSpace,in Default,out Value,out Found)"}.Value
    OffsetTransform.Rotation=RigVMDispatchNode(True=(X=0,Y=0,Z=0,W=1), False=@n50.Result, Condition=RigVMVariableNode() {variable=IK Compensate World Orient}.Value) {template="DISPATCH_RigVMDispatch_If(in Condition,in True,in False,out Result)"}.Result
    InitialValue.Rotation=@n50.Result
    InitialValue.Translation=@n49.Transform.Translation
    OffsetTransform.Translation=@n49.Transform.Translation
    Settings.Shape.Name=@n51.Value.Name
    Settings.Shape.Transform=@n51.Value.Transform
    Settings.PreferredRotationOrder=RigVMVariableNode() {variable=Rotation Order}.Value
    Settings.Shape.Color=@"Control Color Override From Metadata_5".Result
   ) {template="SpawnControl::Execute(in InitialValue,in Settings,in OffsetTransform,in OffsetSpace,in Parent,in Name,out Item)"}
   > RigVMVariableNode(Value=@n52.Item) {variable=IK Control}
   n53 = &RigUnit_HierarchyGetShapeSettings(Item=@n52.Item)
   > Create Gimbal Control_1 = RigVMFunctionReferenceNode(
    Shape Settings Override=(bVisible=True,Name="DefaultGizmoLibraryNormalized.Sphere_Thin",Color=(R=1,G=0,B=0,A=1),Transform=(Rotation=(X=0,Y=0,Z=0,W=1),Translation=(X=0,Y=0,Z=0),Scale3D=(X=1,Y=1,Z=1)))
    Shape Settings Override.Color=/Script/RigVM.RigVMFunction_MathColorAdd(B=(R=0.2,G=0.2,B=0.2,A=1), A=@n53.Settings.Color) {is=struct RigVMUnitNode, template="Add::Execute(in A,in B,out Result)"}.Result
    Shape Settings Override.Transform=@n53.Settings.Transform
    Shape Settings Override.bVisible=@n53.Settings.bVisible
    Rotation Order=RigVMVariableNode() {variable=Rotation Order}.Value
    Source Control=@n52.Item
   )
   > RigVMVariableNode(Value=@"Create Gimbal Control_1".Gimbal Control) {variable=IK Gimbal Control}
   > n54 = &RigUnit_HierarchyAddNull(
    Name=IK
    Space=LocalSpace
    Transform=(Rotation=(X=0,Y=0,Z=0,W=1),Translation=(X=0,Y=0,Z=0),Scale3D=(X=1,Y=1,Z=1))
    Parent=@"Create Gimbal Control_1".Gimbal Control
   )
   > RigVMVariableNode(Value=@n54.Item) {variable=IK Null}
   > DISPATCH_RigDispatch_SetModuleMetadata = RigVMDispatchNode(Name=IK Null, NameSpace=Self, Value=@n54.Item) {template="DISPATCH_RigDispatch_SetModuleMetadata(in Name,in NameSpace,in Value,out Success)"}
   > RigVMDispatchNode(Name=IK Control, NameSpace=Self, Value=@n52.Item) {template="DISPATCH_RigDispatch_SetModuleMetadata(in Name,in NameSpace,in Value,out Success)"}
   > n55 = RigVMDispatchNode(Name=IK Driver, NameSpace=Self, Value=@"Create Gimbal Control_1".Gimbal Control) {template="DISPATCH_RigDispatch_SetModuleMetadata(in Name,in NameSpace,in Value,out Success)"}
   n56 = &RigVMFunction_ControlFlowBranch(ExecuteContext=@n55.ExecuteContext, Condition=RigVMVariableNode() {variable=Hand Align}.Value)
   Control Color Override From Metadata_4 = RigVMFunctionReferenceNode(Color Override=RigVMVariableNode() {variable=IK Control Shape}.Value.Color)
   Multiply = /Script/RigVM.RigVMFunction_MathColorMul(B=(R=0.8,G=0.8,B=0.5,A=1), A=@"Control Color Override From Metadata_4".Result) {is=struct RigVMUnitNode, template="Multiply::Execute(in A,in B,out Result)"}
   n57 = &RigUnit_GetTransform(
    bInitial=true
    Item=RigVMDispatchNode(Index=-1, Array=RigVMVariableNode() {variable=Virtual Bones}.Value) {template="DISPATCH_RigVMDispatch_ArrayGetAtIndex(in Array,in Index,out Element)"}.Element
   )
   n58 = /Script/ControlRig.RigUnit_HierarchyAddControlRotator(
    InitialValue=(Pitch=0,Yaw=0,Roll=0)
    Name=IK Rotation
    OffsetSpace=GlobalSpace
    OffsetTransform=(Rotation=(X=0,Y=0,Z=0,W=1),Translation=(X=0,Y=0,Z=0),Scale3D=(X=1,Y=1,Z=1))
    Settings="(InitialSpace=LocalSpace,Limits=(LimitPitch=(bMinimum=False,bMaximum=False),LimitYaw=(bMinimum=False,bMaximum=False),LimitRoll=(bMinimum=False,bMaximum=False),MinValue=(Pitch=-180,Yaw=-180,Roll=-180),MaxValue=(Pitch=180,Yaw=180,Roll=180),bDrawLimits=True), ...(truncated)"
    Parent=RigVMDispatchNode(Default=(Type=Bone,Name="None"), Name=Local Control, NameSpace=Root) {template="DISPATCH_RigDispatch_GetModuleMetadata(in Name,in NameSpace,in Default,out Value,out Found)"}.Value
    Settings.Shape.Color=@Multiply.Result
    OffsetTransform.Rotation=&RigVMFunction_MathQuaternionMul(A=@n57.Transform.Rotation, B=RigVMVariableNode() {variable=IK Rotation Offset}.Value) {template="Multiply::Execute(in A,in B,out Result)"}.Result
    OffsetTransform.Translation=@n57.Transform.Translation
    ExecutePin=@n56.True
   ) {
    is=struct RigVMUnitNode
    template="SpawnControl::Execute(in InitialValue,in Settings,in OffsetTransform,in OffsetSpace,in Parent,in Name,out Item)"
   }
   > RigVMVariableNode(Value=@n58.Item) {variable=IK Rotation Control}
   > n59 = &RigUnit_HierarchyAddNull(
    Name=IK Rotation Null
    Space=GlobalSpace
    Parent=RigVMDispatchNode(Index=1, Array=RigVMVariableNode() {variable=Virtual Bones}.Value) {template="DISPATCH_RigVMDispatch_ArrayGetAtIndex(in Array,in Index,out Element)"}.Element
    Transform=&RigUnit_GetTransform(bInitial=true, Item=@n58.Item).Transform
   )
   > RigVMVariableNode(Value=@n59.Item) {variable=IK Rotation Null}
   n60 = &RigUnit_GetTransform(
    bInitial=true
    Item=RigVMDispatchNode(Index=-1, Array=RigVMVariableNode() {variable=Virtual Bones}.Value) {template="DISPATCH_RigVMDispatch_ArrayGetAtIndex(in Array,in Index,out Element)"}.Element
   )
   n61 = RigVMVariableNode() {variable=Virtual Bones}
   n62 = &RigUnit_HierarchyAddNull(
    Name=PV Twist Orient Space
    Space=GlobalSpace
    Transform=(Rotation=(X=0,Y=0,Z=0,W=1),Translation=(X=0,Y=0,Z=0),Scale3D=(X=1,Y=1,Z=1))
    Transform.Scale3D=@n60.Transform.Scale3D
    Transform.Translation=@n60.Transform.Translation
    Transform.Rotation=&RigUnit_GetTransform(
     bInitial=true
     Item=RigVMDispatchNode(Index=1, Array=@n61.Value) {template="DISPATCH_RigVMDispatch_ArrayGetAtIndex(in Array,in Index,out Element)"}.Element
    ).Transform.Rotation
    Parent=RigVMVariableNode() {variable=IK Control}.Value
    ExecutePin=@RigVMFunction_Sequence.E
   )
   > RigVMVariableNode(Value=@n62.Item) {variable=PV Twist Orient Space}
   n63 = RigVMVariableNode() {variable=IK Control}
   GetAnimationChannel = &RigUnit_GetFloatAnimationChannel(Channel=PV Twist Follow, bInitial=true, Control=@n63.Value.Name) {template="GetAnimationChannel::Execute(out Value,in Control,in Channel,in bInitial)"}
   > Compute Pole Vector Parent = RigVMFunctionReferenceNode(
    Initial=true
    Lower Segment Scale=1
    Upper Segment Scale=1
    Root Bone=RigVMDispatchNode(Index=0, Array=RigVMVariableNode() {variable=Virtual Bones}.Value) {template="DISPATCH_RigVMDispatch_ArrayGetAtIndex(in Array,in Index,out Element)"}.Element
    PV Twist Follow=@GetAnimationChannel.Value
    IK Control=@n63.Value
    Lengths=RigVMVariableNode() {variable=Default Lengths}.Value
    Negative Side=RigVMVariableNode() {variable=Negative Side}.Value
   )
   > n64 = &RigUnit_HierarchyAddNull(Name="Auto PV ", Parent=(Type=Bone,Name="None"), Space=GlobalSpace, Transform=@"Compute Pole Vector Parent".Result)
   > RigVMVariableNode(Value=@n64.Item) {variable=Auto PV Parent}
   Compute PV Position_1 = RigVMFunctionReferenceNode(
    Distance=RigVMDispatchNode(Index=0, Array=RigVMVariableNode() {variable=Default Lengths}.Value) {template="DISPATCH_RigVMDispatch_ArrayGetAtIndex(in Array,in Index,out Element)"}.Element
    Distance Scale=RigVMVariableNode() {variable=PV Distance Scale}.Value
    Items=RigVMVariableNode() {variable=Virtual Bones}.Value
    Negative Side=RigVMVariableNode() {variable=Negative Side}.Value
   )
   > n65 = &RigUnit_HierarchyAddNull(
    Name="Orient PV "
    Space=GlobalSpace
    Transform=(Rotation=(X=0,Y=0,Z=0,W=1),Translation=(X=0,Y=0,Z=0),Scale3D=(X=1,Y=1,Z=1))
    Transform.Translation=@"Compute PV Position_1".Result
    Parent=@n64.Item
   )
   n66 = RigVMVariableNode() {variable=PV Control Shape}
   Control Color Override From Metadata_2 = RigVMFunctionReferenceNode(Color Override=@n66.Value.Color)
   > n67 = /Script/ControlRig.RigUnit_HierarchyAddControlVector(
    InitialValue=(X=0,Y=0,Z=0)
    Name=PV
    OffsetSpace=LocalSpace
    OffsetTransform=(Rotation=(X=0,Y=0,Z=0,W=1),Translation=(X=0,Y=0,Z=0),Scale3D=(X=1,Y=1,Z=1))
    Settings="(InitialSpace=LocalSpace,bIsPosition=True,Limits=(LimitX=(bMinimum=False,bMaximum=False),LimitY=(bMinimum=False,bMaximum=False),LimitZ=(bMinimum=False,bMaximum=False),MinValue=(X=0,Y=0,Z=0),MaxValue=(X=1,Y=1,Z=1),bDrawLimits=True),Shape=(bVisible=True,Name ...(truncated)"
    Settings.Shape.Color=@"Control Color Override From Metadata_2".Result
    Parent=@n65.Item
    Settings.Shape.Name=@n66.Value.Name
    Settings.Shape.Transform=@n66.Value.Transform
    Settings.Shape.bVisible=@n66.Value.bVisible
   ) {
    is=struct RigVMUnitNode
    template="SpawnControl::Execute(in InitialValue,in Settings,in OffsetTransform,in OffsetSpace,in Parent,in Name,out Item)"
   }
   > RigVMVariableNode(Value=@n67.Item) {variable=PV Control}
   n68 = RigVMVariableNode() {variable=Virtual Bones}
   > n69 = &RigUnit_HierarchyAddNull(
    Name=Start Twist PV
    Space=GlobalSpace
    Transform=&RigUnit_GetTransform(
     bInitial=true
     Item=RigVMDispatchNode(Index=0, Array=@n68.Value) {template="DISPATCH_RigVMDispatch_ArrayGetAtIndex(in Array,in Index,out Element)"}.Element
    ).Transform
    Parent=RigVMVariableNode() {variable=Parent}.Value
   )
   > RigVMVariableNode(Value=@n69.Item) {variable=Start Twist PV}
   > n70 = RigVMDispatchNode(Array=RigVMVariableNode() {variable=IK Spaces}.Value, Element=RigVMVariableNode() {variable=IK Control}.Value) {template="DISPATCH_RigVMDispatch_ArrayAdd(io Array,in Element,out Index)"}
   > For_Each_4 = RigVMDispatchNode(Array=@n70.Array) {template="DISPATCH_RigVMDispatch_ArrayIterator(in Array,out Element,out Index,out Count,out Ratio)"}
   Not_Equals = RigVMDispatchNode(A=&RigUnit_ResolveConnector(Connector=(Type=Connector,Name="End"), SkipSocket=False).Result, B=@For_Each_4.Element) {template="DISPATCH_RigVMDispatch_CoreNotEquals(in A,in B,out Result)"}
   n71 = &RigVMFunction_ControlFlowBranch(ExecuteContext=@For_Each_4.ExecuteContext, Condition=@Not_Equals.Result)
   Not_Equals_2 = RigVMDispatchNode(B=RigVMVariableNode() {variable=IK Control}.Value, A=@For_Each_4.Element) {template="DISPATCH_RigVMDispatch_CoreNotEquals(in A,in B,out Result)"}
   n72 = &RigVMFunction_ControlFlowBranch(Condition=@Not_Equals_2.Result, ExecuteContext=@n71.True)
   n73 = &RigUnit_HierarchyAddNull(
    Space=GlobalSpace
    Name=&RigVMFunction_NameConcat(B=" PV Null", A=@For_Each_4.Element.Name) {template="Concat::Execute(in A,in B,out Result)"}.Result
    Transform=&RigUnit_GetTransform(bInitial=true, Item=RigVMVariableNode() {variable=PV Control}.Value).Transform
    ExecutePin=@n72.Completed
    Parent=@For_Each_4.Element
   )
   > /Script/ControlRig.RigUnit_AddAvailableSpaces(
    Spaces=((Key=(Type=Control,Name="None"),Label="None"))
    Control=RigVMVariableNode() {variable=PV Control}.Value
    Spaces.0.Key=@n73.Item
    Spaces.0.Label=@For_Each_4.Element.Name
   ) {is=struct RigVMUnitNode}
   n74 = RigVMVariableNode() {variable=IK Control}
   n75 = &RigUnit_HierarchyAddNull(
    Space=GlobalSpace
    Name=&RigVMFunction_NameConcat(B=" IK Null", A=@For_Each_4.Element.Name) {template="Concat::Execute(in A,in B,out Result)"}.Result
    Transform=&RigUnit_GetTransform(bInitial=true, Item=&RigUnit_HierarchyGetParent(bDefaultParent=True, Child=@n74.Value).Parent).Transform
    ExecutePin=@n72.True
    Parent=@For_Each_4.Element
   )
   > &RigUnit_AddAvailableSpaces(
    Spaces=((Key=(Type=Control,Name="None"),Label="None"))
    Control=RigVMVariableNode() {variable=IK Control}.Value
    Spaces.0.Key=@n75.Item
    Spaces.0.Label=@For_Each_4.Element.Name
   )
   n76 = RigVMVariableNode() {variable=Bones}
   Create IK Plane Virtual Bones = RigVMFunctionReferenceNode(
    Axis=(X=1,Y=0,Z=0)
    Start=RigVMDispatchNode(Index=0, Array=@n76.Value) {template="DISPATCH_RigVMDispatch_ArrayGetAtIndex(in Array,in Index,out Element)"}.Element
    Mid=RigVMDispatchNode(Index=1, Array=@n76.Value) {template="DISPATCH_RigVMDispatch_ArrayGetAtIndex(in Array,in Index,out Element)"}.Element
    End=RigVMDispatchNode(Index=2, Array=@n76.Value) {template="DISPATCH_RigVMDispatch_ArrayGetAtIndex(in Array,in Index,out Element)"}.Element
    Negative Side=RigVMVariableNode() {variable=Negative Side}.Value
    ExecuteContext=@RigVMFunction_Sequence.A
   )
   > RigVMVariableNode(Value=@"Create IK Plane Virtual Bones".Bones) {variable=Virtual Bones}
   n77 = RigVMDispatchNode(Index=-1, Array=@"Create IK Plane Virtual Bones".Bones) {template="DISPATCH_RigVMDispatch_ArrayGetAtIndex(in Array,in Index,out Element)"}
   > Set Transform = &RigUnit_SetRotation(
    Weight=1
    bInitial=true
    bPropagateToChildren=false
    Value=&RigVMFunction_MathQuaternionMul(
     A=&RigUnit_GetTransform(Space=LocalSpace, bInitial=true, Item=@n77.Element).Transform.Rotation
     B=RigVMVariableNode() {variable=End Bone Rotation Offset}.Value
    ) {template="Multiply::Execute(in A,in B,out Result)"}.Result
    Item=@n77.Element
   )
   n78 = &RigUnit_GetTransform(Item=(Type=Bone,Name="None"), bInitial=true)
   RigVMVariableNode(ExecuteContext=(), Value=@n78.Transform) {variable=IKTransformInitial}
   RigVMFunction_MathQuaternionSwingTwist = &RigVMFunction_MathQuaternionSwingTwist(TwistAxis=(X=0,Y=0,Z=1), Input=@n78.Transform.Rotation)
   > RigVMVariableNode(Value=@RigVMFunction_MathQuaternionSwingTwist.Twist) {variable=IKTwistInitial}
   &RigUnit_GetTransform(bInitial=true, Item=@n1.Value)
   RigVMVariableNode() {variable=IK Twist Null}
  ]}
 RigVMModel Forward Solve Graph {nodes [
   n1 = /Script/ControlRig.RigUnit_PreBeginExecution() {is=struct RigVMUnitNode}
   n2 = &RigUnit_GetBoolAnimationChannelFromItem(bInitial=False, Item=RigVMVariableNode() {variable=Ik Solve Channel}.Value) {template="GetAnimationChannelFromItem::Execute(out Value,in Item,in bInitial)"}
   RigVMVariableNode(Value=@n2.Value, ExecuteContext=@n1.ExecutePin) {variable=IK Solve}
   > RigVMDispatchNode(Name=IK Solve, NameSpace=Self, Value=@n2.Value) {template="DISPATCH_RigDispatch_SetModuleMetadata(in Name,in NameSpace,in Value,out Success)"}
   RigUnit_BeginExecution = &RigUnit_BeginExecution()
   RigUnit_ItemToName = &RigUnit_ItemToName(Value=RigVMVariableNode() {variable=IK Control}.Value) {template="Cast::Execute(in Value,out Result)"}
   RigVMVariableNode(
    Value=/Script/ControlRig.RigUnit_GetBoolAnimationChannel(Channel=Stretch, bInitial=False, Control=@RigUnit_ItemToName.Result) {is=struct RigVMUnitNode, template="GetAnimationChannel::Execute(out Value,in Control,in Channel,in bInitial)"}.Value
    ExecuteContext=@RigUnit_BeginExecution.ExecutePin
   ) {variable=Stretch}
   RigUnit_GetFloatAnimationChannel = &RigUnit_GetFloatAnimationChannel(Channel=Softness, bInitial=False, Control=@RigUnit_ItemToName.Result) {template="GetAnimationChannel::Execute(out Value,in Control,in Channel,in bInitial)"}
   > n3 = RigVMVariableNode(Value=@RigUnit_GetFloatAnimationChannel.Value) {variable=Softness}
   n4 = &RigVMFunction_ControlFlowBranch(ExecuteContext=@n3.ExecuteContext, Condition=RigVMVariableNode() {variable=Segment Scale Control}.Value)
   RigVMFunction_Sequence = RigVMAggregateNode(ExecuteContext=@n4.Completed)
   RigVMFunction_ControlFlowBranch = &RigVMFunction_ControlFlowBranch(ExecuteContext=@RigVMFunction_Sequence.C, Condition=RigVMVariableNode() {variable=IK Solve}.Value)
   n5 = RigVMVariableNode() {variable=FK Controls}
   /Script/ControlRig.RigUnit_SetScale(
    Scale=(X=1,Y=1,Z=1)
    Space=LocalSpace
    Weight=1
    bInitial=False
    bPropagateToChildren=True
    Item=RigVMDispatchNode(Index=0, Array=@n5.Value) {template="DISPATCH_RigVMDispatch_ArrayGetAtIndex(in Array,in Index,out Element)"}.Element
    ExecutePin=@RigVMFunction_ControlFlowBranch.False
   ) {is=struct RigVMUnitNode}
   > &RigUnit_SetScale(
    Scale=(X=1,Y=1,Z=1)
    Space=LocalSpace
    Weight=1
    bInitial=False
    bPropagateToChildren=True
    Item=RigVMDispatchNode(Index=1, Array=@n5.Value) {template="DISPATCH_RigVMDispatch_ArrayGetAtIndex(in Array,in Index,out Element)"}.Element
   )
   n6 = RigVMVariableNode() {variable=FK Controls}
   n7 = &RigUnit_HierarchyGetParent(
    bDefaultParent=True
    Child=RigVMDispatchNode(Index=1, Array=@n6.Value) {template="DISPATCH_RigVMDispatch_ArrayGetAtIndex(in Array,in Index,out Element)"}.Element
   )
   > Set Transform_5 = &RigUnit_SetTranslation(
    Space=LocalSpace
    Value=(X=0,Y=0,Z=0)
    Weight=1
    bInitial=False
    bPropagateToChildren=True
    Item=@n7.Parent
    Value.X=&RigVMFunction_MathDoubleMul(
     A=&RigUnit_GetTransform(Space=LocalSpace, bInitial=true, Item=@n7.Parent).Transform.Translation.X
     B=RigVMVariableNode() {variable=Upper Segment Scale}.Value
    ).Result
   ) {template="Set Transform::Execute(in Item,in Space,in bInitial,in Value,in Weight,in bPropagateToChildren)"}
   n8 = &RigUnit_HierarchyGetParent(
    bDefaultParent=True
    Child=RigVMDispatchNode(Index=2, Array=@n6.Value) {template="DISPATCH_RigVMDispatch_ArrayGetAtIndex(in Array,in Index,out Element)"}.Element
   )
   > Set Transform_1 = &RigUnit_SetTranslation(
    Space=LocalSpace
    Value=(X=0,Y=0,Z=0)
    Weight=1
    bInitial=False
    bPropagateToChildren=True
    Item=@n8.Parent
    Value.X=&RigVMFunction_MathDoubleMul(
     A=&RigUnit_GetTransform(Space=LocalSpace, bInitial=true, Item=@n8.Parent).Transform.Translation.X
     B=RigVMVariableNode() {variable=Lower Segment Scale}.Value
    ).Result
   ) {template="Set Transform::Execute(in Item,in Space,in bInitial,in Value,in Weight,in bPropagateToChildren)"}
   > Compute FK_1 = RigVMFunctionReferenceNode(
    Transforms=&RigUnit_GetTransformItemArray(Space=GlobalSpace, Items=RigVMVariableNode() {variable=FK Drivers}.Value).Transforms
    Negative Side=RigVMVariableNode() {variable=Negative Side}.Value
    Length=RigVMVariableNode() {variable=Default Lengths}.Value
    Apply Scale=RigVMVariableNode() {variable=Use Scale}.Value
   )
   > For_Each_1 = RigVMDispatchNode(Array=@"Compute FK_1".Result) {template="DISPATCH_RigVMDispatch_ArrayIterator(in Array,out Element,out Index,out Count,out Ratio)"}
   Inverse = &RigVMFunction_MathQuaternionInverse(Value=RigVMVariableNode() {variable=FK Rotation Offset}.Value) {template="Inverse::Execute(in Value,out Result)"}
   > &RigUnit_SetTransform(
    Value=(Rotation=(X=0,Y=0,Z=0,W=1),Translation=(X=0,Y=0,Z=0),Scale3D=(X=1,Y=1,Z=1))
    Weight=1
    bInitial=False
    Item=RigVMDispatchNode(Index=@For_Each_1.Index, Array=RigVMVariableNode() {variable=Virtual Bones}.Value) {template="DISPATCH_RigVMDispatch_ArrayGetAtIndex(in Array,in Index,out Element)"}.Element
    Value.Scale3D=@For_Each_1.Element.Scale3D
    Value.Translation=@For_Each_1.Element.Translation
    Value.Rotation=&RigVMFunction_MathQuaternionMul(A=@For_Each_1.Element.Rotation, B=@Inverse.Result) {template="Multiply::Execute(in A,in B,out Result)"}.Result
   )
   Match IK = RigVMFunctionReferenceNode(ExecuteContext=@For_Each_1.Completed)
   Set Transform_7 = &RigUnit_SetTranslation(
    Space=GlobalSpace
    Weight=1
    bInitial=False
    bPropagateToChildren=True
    Item=RigVMDispatchNode(Index=0, Array=RigVMVariableNode() {variable=Virtual Bones}.Value) {template="DISPATCH_RigVMDispatch_ArrayGetAtIndex(in Array,in Index,out Element)"}.Element
    Value=&RigUnit_GetTransform(bInitial=False, Item=RigVMVariableNode() {variable=IK Base Control}.Value).Transform.Translation
    ExecutePin=@RigVMFunction_ControlFlowBranch.True
   ) {template="Set Transform::Execute(in Item,in Space,in bInitial,in Value,in Weight,in bPropagateToChildren)"}
   > Compute Pole Vector Parent = RigVMFunctionReferenceNode(
    Initial=False
    PV Twist Follow=&RigUnit_GetFloatAnimationChannel(Channel=PV Twist Follow, bInitial=False, Control=RigVMVariableNode() {variable=IK Control}.Value.Name) {template="GetAnimationChannel::Execute(out Value,in Control,in Channel,in bInitial)"}.Value
    Root Bone=RigVMVariableNode() {variable=Start Twist PV}.Value
    Lengths=RigVMVariableNode() {variable=Default Lengths}.Value
    IK Control=RigVMVariableNode() {variable=PV Twist Orient Space}.Value
    Negative Side=RigVMVariableNode() {variable=Negative Side}.Value
    Upper Segment Scale=RigVMVariableNode() {variable=Upper Segment Scale}.Value
    Debug=RigVMVariableNode() {variable=Debug}.Value
    Lower Segment Scale=RigVMVariableNode() {variable=Lower Segment Scale}.Value
   )
   > &RigUnit_SetTransform(
    Weight=1
    bInitial=False
    bPropagateToChildren=true
    Value=@"Compute Pole Vector Parent".Result
    Item=RigVMVariableNode() {variable=Auto PV Parent}.Value
   )
   n9 = RigVMVariableNode() {variable=Virtual Bones}
   DISPATCH_RigVMDispatch_If = RigVMDispatchNode(False=1, True=-1, Condition=RigVMVariableNode() {variable=Negative Side}.Value) {template="DISPATCH_RigVMDispatch_If(in Condition,in True,in False,out Result)"}
   Negate = &RigVMFunction_MathDoubleNegate(Value=@DISPATCH_RigVMDispatch_If.Result) {template="Negate::Execute(in Value,out Result)"}
   > Soft IK = RigVMFunctionReferenceNode(
    PoleVector=(X=0,Y=0,Z=0)
    PoleVectorKind=Location
    PrimaryAxis=(X=1,Y=0,Z=0)
    SecondaryAxis=(X=0,Y=-1,Z=0)
    SecondaryAxisWeight=1
    Softness Distance=0.95
    StretchMaximumRatio=4
    StretchStartRatio=1
    Weight=1
    Bone A=RigVMDispatchNode(Index=0, Array=@n9.Value) {template="DISPATCH_RigVMDispatch_ArrayGetAtIndex(in Array,in Index,out Element)"}.Element
    Bone B=RigVMDispatchNode(Index=1, Array=@n9.Value) {template="DISPATCH_RigVMDispatch_ArrayGetAtIndex(in Array,in Index,out Element)"}.Element
    Bone C=RigVMDispatchNode(Index=2, Array=@n9.Value) {template="DISPATCH_RigVMDispatch_ArrayGetAtIndex(in Array,in Index,out Element)"}.Element
    Effector=&RigUnit_GetTransform(bInitial=False, Item=RigVMVariableNode() {variable=IK Null}.Value).Transform
    SecondaryAxis.Y=@Negate.Result
    PoleVectorSpace=RigVMVariableNode() {variable=PV Control}.Value
    EnableStretch=RigVMVariableNode() {variable=Stretch}.Value
    Softness=RigVMVariableNode() {variable=Softness}.Value
    Upper Segment Scale=RigVMVariableNode() {variable=Upper Segment Scale}.Value
    Lower Segment Scale=RigVMVariableNode() {variable=Lower Segment Scale}.Value
    Debug=RigVMVariableNode() {variable=Debug}.Value
    PrimaryAxis.X=@DISPATCH_RigVMDispatch_If.Result
   )
   n10 = RigVMVariableNode() {variable=Virtual Bones}
   n11 = RigVMVariableNode() {variable=IK Base Control}
   > &RigUnit_SetTransform(
    Value=(Rotation=(X=0,Y=0,Z=0,W=1),Translation=(X=0,Y=0,Z=0),Scale3D=(X=1,Y=1,Z=1))
    Weight=1
    bInitial=False
    Value.Rotation=&RigUnit_GetTransform(
     bInitial=False
     Item=RigVMDispatchNode(Index=0, Array=@n10.Value) {template="DISPATCH_RigVMDispatch_ArrayGetAtIndex(in Array,in Index,out Element)"}.Element
    ).Transform.Rotation
    Value.Translation=&RigUnit_GetTransform(bInitial=False, Item=@n11.Value).Transform.Translation
    Item=@n11.Value
   )
   > RigVMFunction_DebugLineNoSpace = &RigVMFunction_DebugLineNoSpace(
    Color=(R=0.2,G=0.2,B=0.2,A=1)
    DebugDrawSettings=(DepthPriority=SDPG_Foreground,Lifetime=-1)
    Thickness=0
    WorldOffset=(Rotation=(X=0,Y=0,Z=0,W=1),Translation=(X=0,Y=0,Z=0),Scale3D=(X=1,Y=1,Z=1))
    bEnabled=True
    A=&RigUnit_GetTransform(bInitial=False, Item=RigVMVariableNode() {variable=Mid Control}.Value).Transform.Translation
    B=&RigUnit_GetTransform(bInitial=False, Item=RigVMVariableNode() {variable=PV Control}.Value).Transform.Translation
   )
   > /Script/ControlRig.RigUnit_SetControlVisibility(
    bVisible=&RigUnit_GetBoolAnimationChannelFromItem(bInitial=False, Item=RigVMVariableNode() {variable=Hand Align Channel}.Value) {template="GetAnimationChannelFromItem::Execute(out Value,in Item,in bInitial)"}.Value
    Item=RigVMVariableNode() {variable=IK Rotation Control}.Value
   ) {is=struct RigVMUnitNode, template="SetControlVisibility::Execute(in Item,in Pattern,in bVisible)"}
   > Hand Align Solve = RigVMFunctionReferenceNode()
   > Match FK = RigVMFunctionReferenceNode()
   > Get Distances Between_0 = RigVMFunctionReferenceNode(Initial=False, Items=RigVMVariableNode() {variable=Virtual Bones}.Value)
   > RigVMVariableNode(Value=@"Get Distances Between_0".Distances) {variable=Current Lengths}
   n12 = &RigVMFunction_Sequence(ExecuteContext=@RigVMFunction_Sequence.A)
   n13 = RigVMVariableNode() {variable=Parent Buffer}
   n14 = RigVMVariableNode() {variable=Parent}
   n15 = &RigUnit_ProjectTransformToNewParent(bChildInitial=True, bNewParentInitial=False, bOldParentInitial=True, Child=@n13.Value, NewParent=@n14.Value, OldParent=@n14.Value)
   &RigUnit_SetTransform(
    Value=(Rotation=(X=0,Y=0,Z=0,W=1),Translation=(X=0,Y=0,Z=0),Scale3D=(X=1,Y=1,Z=1))
    Weight=1
    bInitial=False
    Value.Rotation=@n15.Transform.Rotation
    Value.Translation=@n15.Transform.Translation
    ExecutePin=@n12.A
    Item=@n13.Value
   )
   For_Each_4 = RigVMDispatchNode(ExecuteContext=@n12.B, Array=RigVMVariableNode() {variable=FK Spaces}.Value) {template="DISPATCH_RigVMDispatch_ArrayIterator(in Array,out Element,out Index,out Count,out Ratio)"}
   n16 = RigVMDispatchNode(Index=@For_Each_4.Index, Array=RigVMVariableNode() {variable=FK Nulls}.Value) {template="DISPATCH_RigVMDispatch_ArrayGetAtIndex(in Array,in Index,out Element)"}
   Multiply = &RigVMFunction_MathTransformMul(
    B=&RigUnit_GetTransform(bInitial=False, Item=@For_Each_4.Element).Transform
    A=&RigVMFunction_MathTransformMakeRelative(
     Parent=&RigUnit_GetTransform(bInitial=true, Item=@For_Each_4.Element).Transform
     Global=&RigUnit_GetTransform(bInitial=true, Item=@n16.Element).Transform
    ).Local
   ) {template="Multiply::Execute(in A,in B,out Result)"}
   > Set Transform_2 = &RigUnit_SetRotation(Space=GlobalSpace, Weight=1, bInitial=False, Item=@n16.Element, Value=@Multiply.Result.Rotation)
   Not_Equals = RigVMDispatchNode(B=RigVMVariableNode() {variable=Previous IK Solve}.Value, A=RigVMVariableNode() {variable=IK Solve}.Value) {template="DISPATCH_RigVMDispatch_CoreNotEquals(in A,in B,out Result)"}
   RigVMFunction_MathBoolAnd = &RigVMFunction_MathBoolAnd(B=@Not_Equals.Result, A=&RigVMFunction_MathBoolNot(Value=RigVMVariableNode() {variable=IK FK Manipulation}.Value).Result)
   n17 = &RigVMFunction_ControlFlowBranch(Condition=@RigVMFunction_MathBoolAnd.Result, ExecuteContext=@RigVMFunction_Sequence.F)
   RigVMVariableNode(ExecuteContext=@n17.Completed, Value=RigVMVariableNode() {variable=IK Solve}.Value) {variable=Previous IK Solve}
   Clone = RigVMDispatchNode(Array=RigVMVariableNode() {variable=FK Controls}.Value) {template="DISPATCH_RigVMDispatch_ArrayClone(in Array,out Clone)"}
   n18 = RigVMDispatchNode(Array=@Clone.Clone, Other=RigVMVariableNode() {variable=FK Gimbal Controls}.Value, ExecuteContext=@n17.True) {template="DISPATCH_RigVMDispatch_ArrayAppend(io Array,in Other)"}
   n19 = RigVMDispatchNode(
    Values=((Type=None,Name="None"),((),Type=None,Name="None"),((),Type=None,Name="None"),((),Type=None,Name="None"),((),Type=None,Name="None"))
    Values.0=RigVMVariableNode() {variable=IK Control}.Value
    Values.1=RigVMVariableNode() {variable=PV Control}.Value
    Values.2=RigVMVariableNode() {variable=IK Gimbal Control}.Value
    Values.3=RigVMVariableNode() {variable=IK Base Control}.Value
    Values.4=RigVMVariableNode() {variable=IK Rotation Control}.Value
   ) {template="DISPATCH_RigVMDispatch_ArrayMake(in Values,out Array)"}
   > n20 = RigVMDispatchNode(Array=@n18.Array, Other=@n19.Array) {template="DISPATCH_RigVMDispatch_ArrayAppend(io Array,in Other)"}
   > For_Each_2 = RigVMDispatchNode(Array=@n20.Array) {template="DISPATCH_RigVMDispatch_ArrayIterator(in Array,out Element,out Index,out Count,out Ratio)"}
   > SendEvent = &RigUnit_SendEvent(Event=RequestAutoKey, OffsetInSeconds=0, bEnable=True, bOnlyDuringInteraction=true, Item=@For_Each_2.Element) {template="SendEvent::Execute(in Event,in Item,in OffsetInSeconds,in bEnable,in bOnlyDuringInteraction)"}
   n21 = &RigUnit_SendEvent(
    Event=RequestAutoKey
    OffsetInSeconds=0
    bEnable=True
    bOnlyDuringInteraction=true
    ExecutePin=@For_Each_2.Completed
    Item=RigVMVariableNode() {variable=Ik Solve Channel}.Value
   ) {template="SendEvent::Execute(in Event,in Item,in OffsetInSeconds,in bEnable,in bOnlyDuringInteraction)"}
   n22 = &RigVMFunction_ControlFlowBranch(
    Condition=&RigVMFunction_MathBoolAnd(
     B=/Script/ControlRig.RigUnit_IsInteracting() {is=struct RigVMUnitNode}.bIsInteracting
     A=RigVMVariableNode() {variable=Segment Scale Control}.Value
    ).Result
    ExecuteContext=@n21.ExecutePin
   )
   "Calculate And Set Segment Scale Channel  Values" = RigVMFunctionReferenceNode(
    Segment Scale Channels=((Type=None,Name="None"),(Type=None,Name="None"))
    ExecuteContext=@n22.True
    Current Lengths=RigVMVariableNode() {variable=Current Lengths}.Value
    Segment Scale Channels.0=RigVMVariableNode() {variable=Upper Segment Scale Channel}.Value
    Segment Scale Channels.1=RigVMVariableNode() {variable=Lower Segment Scale Channel}.Value
    Default Lengths=RigVMVariableNode() {variable=Default Lengths}.Value
   )
   > &RigUnit_SendEvent(
    Event=RequestAutoKey
    OffsetInSeconds=0
    bEnable=True
    bOnlyDuringInteraction=true
    Item=RigVMVariableNode() {variable=Upper Segment Scale Channel}.Value
   ) {template="SendEvent::Execute(in Event,in Item,in OffsetInSeconds,in bEnable,in bOnlyDuringInteraction)"}
   > &RigUnit_SendEvent(
    Event=RequestAutoKey
    OffsetInSeconds=0
    bEnable=True
    bOnlyDuringInteraction=true
    Item=RigVMVariableNode() {variable=Lower Segment Scale Channel}.Value
   ) {template="SendEvent::Execute(in Event,in Item,in OffsetInSeconds,in bEnable,in bOnlyDuringInteraction)"}
   Compute Mid Null Transform = RigVMFunctionReferenceNode(Items=RigVMVariableNode() {variable=Virtual Bones}.Value)
   &RigUnit_SetTransform(
    Weight=1
    bInitial=False
    Value=@"Compute Mid Null Transform".Struct
    Item=RigVMVariableNode() {variable=Mid Null}.Value
    ExecutePin=@RigVMFunction_Sequence.D
   )
   n23 = &RigUnit_GetTransformItemArray(Space=GlobalSpace, bInitial=False, Items=RigVMVariableNode() {variable=Virtual Bones}.Value)
   > Compute FK = RigVMFunctionReferenceNode(
    Transforms=((Rotation=(X=0,Y=0,Z=0,W=1),Translation=(X=0,Y=0,Z=0),Scale3D=(X=1,Y=1,Z=1)),(Rotation=(X=0,Y=0,Z=0,W=1),Translation=(X=0,Y=0,Z=0),Scale3D=(X=1,Y=1,Z=1)),(Rotation=(X=0,Y=0,Z=0,W=1),Translation=(X=0,Y=0,Z=0),Scale3D=(X=1,Y=1,Z=1)))
    Transforms.0=RigVMDispatchNode(Index=0, Array=@n23.Transforms) {template="DISPATCH_RigVMDispatch_ArrayGetAtIndex(in Array,in Index,out Element)"}.Element
    Transforms.2=RigVMDispatchNode(Index=-1, Array=@n23.Transforms) {template="DISPATCH_RigVMDispatch_ArrayGetAtIndex(in Array,in Index,out Element)"}.Element
    Transforms.1=&RigUnit_GetTransform(bInitial=False, Item=RigVMVariableNode() {variable=Mid Control}.Value).Transform
    Negative Side=RigVMVariableNode() {variable=Negative Side}.Value
    Length=RigVMVariableNode() {variable=Default Lengths}.Value
    Apply Scale=RigVMVariableNode() {variable=Use Scale}.Value
   )
   > /Script/ControlRig.RigUnit_SetTransformItemArray(
    Space=GlobalSpace
    Weight=1
    bInitial=False
    bPropagateToChildren=True
    Transforms=@"Compute FK".Result
    Items=RigVMVariableNode() {variable=Virtual Bones}.Value
   ) {is=struct RigVMUnitNode}
   n24 = RigVMVariableNode() {variable=Debug}
   RigVMFunction_DebugTransformArrayMutableNoSpace = &RigVMFunction_DebugTransformArrayMutableNoSpace(
    Color=(R=1,G=1,B=1,A=1)
    DebugDrawSettings=(DepthPriority=SDPG_Foreground,Lifetime=-1)
    Mode=Axes
    ParentIndices=()
    Scale=10
    Thickness=0
    WorldOffset=(Rotation=(X=0,Y=0,Z=0,W=1),Translation=(X=0,Y=0,Z=0),Scale3D=(X=1,Y=1,Z=1))
    Transforms=&RigUnit_GetTransformItemArray(Space=GlobalSpace, bInitial=false, Items=RigVMVariableNode() {variable=Virtual Bones}.Value).Transforms
    bEnabled=@n24.Value
    ExecutePin=@RigVMFunction_Sequence.G
   )
   > RigVMFunction_DebugTransformMutableNoSpace = &RigVMFunction_DebugTransformMutableNoSpace(
    Color=(R=1,G=0,B=0,A=1)
    DebugDrawSettings=(DepthPriority=SDPG_Foreground,Lifetime=-1)
    Mode=Axes
    Scale=10
    Thickness=1
    WorldOffset=(Rotation=(X=0,Y=0,Z=0,W=1),Translation=(X=0,Y=0,Z=0),Scale3D=(X=1,Y=1,Z=1))
    Transform=&RigUnit_GetTransform(bInitial=False, Item=RigVMVariableNode() {variable=IK Twist Null}.Value).Transform
    bEnabled=@n24.Value
   )
   > &RigVMFunction_DebugTransformMutableNoSpace(
    Color=(R=1,G=0,B=0,A=1)
    DebugDrawSettings=(DepthPriority=SDPG_Foreground,Lifetime=-1)
    Mode=Axes
    Scale=10
    Thickness=0
    WorldOffset=(Rotation=(X=0,Y=0,Z=0,W=1),Translation=(X=0,Y=0,Z=0),Scale3D=(X=1,Y=1,Z=1))
    Transform=&RigUnit_GetTransform(bInitial=False, Item=RigVMVariableNode() {variable=Auto PV Parent}.Value).Transform
    bEnabled=@n24.Value
   )
   For_Each = RigVMDispatchNode(Array=RigVMVariableNode() {variable=Bones}.Value, ExecuteContext=@RigVMFunction_Sequence.E) {template="DISPATCH_RigVMDispatch_ArrayIterator(in Array,out Element,out Index,out Count,out Ratio)"}
   n25 = RigVMDispatchNode(Index=@For_Each.Index, Array=RigVMVariableNode() {variable=Virtual Bones}.Value) {template="DISPATCH_RigVMDispatch_ArrayGetAtIndex(in Array,in Index,out Element)"}
   > &RigUnit_SetTransform(
    Weight=1
    bInitial=False
    Item=@For_Each.Element
    Value=&RigUnit_ProjectTransformToNewParent(bChildInitial=True, bNewParentInitial=False, bOldParentInitial=True, NewParent=@n25.Element, OldParent=@n25.Element, Child=@For_Each.Element).Transform
   )
   Set_Module_Metadata_1 = RigVMDispatchNode(
    Name=IsInteracting
    NameSpace=Self
    Value=&RigUnit_IsInteracting().bIsInteracting
    ExecuteContext=@RigVMFunction_Sequence.H
   ) {template="DISPATCH_RigDispatch_SetModuleMetadata(in Name,in NameSpace,in Value,out Success)"}
   n26 = &RigVMFunction_ControlFlowBranch(Condition=RigVMVariableNode() {variable=IK FK Manipulation}.Value, ExecuteContext=@RigVMFunction_Sequence.B)
   IsInteracting = &RigUnit_IsInteracting()
   n27 = &RigVMFunction_ControlFlowBranch(Condition=@IsInteracting.bIsInteracting, ExecuteContext=@n26.True)
   DISPATCH_RigVMDispatch_ArrayGetAtIndex = RigVMDispatchNode(Index=0, Array=@IsInteracting.Items) {template="DISPATCH_RigVMDispatch_ArrayGetAtIndex(in Array,in Index,out Element)"}
   DISPATCH_RigVMDispatch_ArrayFind = RigVMDispatchNode(Element=@DISPATCH_RigVMDispatch_ArrayGetAtIndex.Element, Array=RigVMVariableNode() {variable=FK Controls}.Value) {template="DISPATCH_RigVMDispatch_ArrayFind(in Array,in Element,out Index,out Success)"}
   n28 = &RigVMFunction_ControlFlowBranch(Condition=@DISPATCH_RigVMDispatch_ArrayFind.Success, ExecuteContext=@n27.True)
   n29 = &RigVMFunction_ControlFlowBranch(
    Condition=RigVMDispatchNode(
     Array=((Type=None,Name="None"),(Type=None,Name="None"))
     Element=@DISPATCH_RigVMDispatch_ArrayGetAtIndex.Element
     Array.0=RigVMVariableNode() {variable=IK Control}.Value
     Array.1=RigVMVariableNode() {variable=PV Control}.Value
    ) {template="DISPATCH_RigVMDispatch_ArrayFind(in Array,in Element,out Index,out Success)"}.Success
    ExecuteContext=@n28.False
   )
   RigVMVariableNode(Value=true, ExecuteContext=@n29.True) {variable=IK Solve}
   RigVMVariableNode(ExecuteContext=@n28.True) {variable=IK Solve}
   RigVMFunction_Sequence_1 = RigVMAggregateNode(ExecuteContext=@n26.False)
   DISPATCH_RigVMDispatch_ArrayIterator = RigVMDispatchNode(ExecuteContext=@RigVMFunction_Sequence_1.A, Array=RigVMVariableNode() {variable=FK Controls}.Value) {template="DISPATCH_RigVMDispatch_ArrayIterator(in Array,out Element,out Index,out Count,out Ratio)"}
   RigVMFunction_MathBoolNot = &RigVMFunction_MathBoolNot(Value=RigVMVariableNode() {variable=IK Solve}.Value)
   > SetControlVisibility = &RigUnit_SetControlVisibility(bVisible=@RigVMFunction_MathBoolNot.Result, Item=@DISPATCH_RigVMDispatch_ArrayIterator.Element) {template="SetControlVisibility::Execute(in Item,in Pattern,in bVisible)"}
   For_Each_3 = RigVMDispatchNode(ExecuteContext=@RigVMFunction_Sequence_1.B, Array=RigVMVariableNode() {variable=FK Gimbal Controls}.Value) {template="DISPATCH_RigVMDispatch_ArrayIterator(in Array,out Element,out Index,out Count,out Ratio)"}
   From_String = RigVMDispatchNode(String=Gimbal Control Vis) {template="DISPATCH_RigDispatch_FromString(in String,out Result)"}
   GetAnimationChannel = &RigUnit_GetBoolAnimationChannel(bInitial=False, Control=@For_Each_3.Element.Name, Channel=@From_String.Result) {template="GetAnimationChannel::Execute(out Value,in Control,in Channel,in bInitial)"}
   n30 = RigVMVariableNode() {variable=IK Solve}
   > &RigUnit_SetControlVisibility(
    Item=@For_Each_3.Element
    bVisible=&RigVMFunction_MathBoolAnd(B=@GetAnimationChannel.Value, A=&RigVMFunction_MathBoolNot(Value=@n30.Value).Result).Result
   ) {template="SetControlVisibility::Execute(in Item,in Pattern,in bVisible)"}
   n31 = RigVMVariableNode() {variable=IK Gimbal Control}
   &RigUnit_SetControlVisibility(
    bVisible=&RigVMFunction_MathBoolAnd(
     B=&RigUnit_GetBoolAnimationChannel(bInitial=False, Control=@n31.Value.Name, Channel=@From_String.Result) {template="GetAnimationChannel::Execute(out Value,in Control,in Channel,in bInitial)"}.Value
     A=RigVMVariableNode() {variable=IK Solve}.Value
    ).Result
    Item=@n31.Value
    ExecutePin=@For_Each_3.Completed
   ) {template="SetControlVisibility::Execute(in Item,in Pattern,in bVisible)"}
   n32 = RigVMVariableNode() {variable=IK Solve}
   &RigUnit_SetControlVisibility(ExecutePin=@RigVMFunction_Sequence_1.C, bVisible=@n32.Value, Item=RigVMVariableNode() {variable=IK Control}.Value) {template="SetControlVisibility::Execute(in Item,in Pattern,in bVisible)"}
   > &RigUnit_SetControlVisibility(Item=RigVMVariableNode() {variable=PV Control}.Value, bVisible=@n32.Value) {template="SetControlVisibility::Execute(in Item,in Pattern,in bVisible)"}
   > &RigUnit_SetControlVisibility(Item=RigVMVariableNode() {variable=IK Rotation Control}.Value, bVisible=@n32.Value) {template="SetControlVisibility::Execute(in Item,in Pattern,in bVisible)"}
   n33 = RigVMVariableNode() {variable=IK Base Control Vis Channel}
   > &RigUnit_SetControlVisibility(
    bVisible=&RigVMFunction_MathBoolAnd(
     B=&RigUnit_GetBoolAnimationChannelFromItem(bInitial=False, Item=@n33.Value) {template="GetAnimationChannelFromItem::Execute(out Value,in Item,in bInitial)"}.Value
     A=@n32.Value
    ).Result
    Item=RigVMVariableNode() {variable=IK Base Control}.Value
   ) {template="SetControlVisibility::Execute(in Item,in Pattern,in bVisible)"}
   GetAnimationChannelFromItem = &RigUnit_GetBoolAnimationChannelFromItem(bInitial=False, Item=RigVMVariableNode() {variable=Sec Controls Vis Channel}.Value) {template="GetAnimationChannelFromItem::Execute(out Value,in Item,in bInitial)"}
   > Set_Module_Metadata = RigVMDispatchNode(Name=Sec Controls Visibilty, NameSpace=Self, Value=@GetAnimationChannelFromItem.Value) {template="DISPATCH_RigDispatch_SetModuleMetadata(in Name,in NameSpace,in Value,out Success)"}
   > &RigUnit_SetControlVisibility(bVisible=@GetAnimationChannelFromItem.Value, Item=RigVMVariableNode() {variable=Mid Control}.Value) {template="SetControlVisibility::Execute(in Item,in Pattern,in bVisible)"}
   RigVMVariableNode(
    Value=/Script/ControlRig.RigUnit_GetFloatAnimationChannelFromItem(bInitial=False, Item=RigVMVariableNode() {variable=Upper Segment Scale Channel}.Value) {is=struct RigVMUnitNode, template="GetAnimationChannelFromItem::Execute(out Value,in Item,in bInitial)"}.Value
    ExecuteContext=@n4.True
   ) {variable=Upper Segment Scale}
   > RigVMVariableNode(Value=&RigUnit_GetFloatAnimationChannelFromItem(bInitial=False, Item=RigVMVariableNode() {variable=Lower Segment Scale Channel}.Value) {template="GetAnimationChannelFromItem::Execute(out Value,in Item,in bInitial)"}.Value) {variable=Lower Segment Scale}
   n34 = &RigVMFunction_Sequence(ExecuteContext=())
   n35 = &RigUnit_GetTransform(bInitial=False, Item=RigVMVariableNode() {variable=IK Control}.Value)
   n36 = RigVMVariableNode() {variable=IK Twist Null}
   &RigUnit_SetTransform(
    Value=(Rotation=(X=0,Y=0,Z=0,W=1),Translation=(X=0,Y=0,Z=0),Scale3D=(X=1,Y=1,Z=1))
    Weight=1
    bInitial=False
    bPropagateToChildren=true
    Value.Scale3D=@n35.Transform.Scale3D
    Value.Translation=@n35.Transform.Translation
    Item=@n36.Value
    Value.Rotation=RigVMVariableNode() {variable=IKTransformInitial}.Value.Rotation
    ExecutePin=@n34.A
   )
   n37 = RigVMVariableNode() {variable=IKTwistInitial}
   RigVMFunction_MathQuaternionSwingTwist = &RigVMFunction_MathQuaternionSwingTwist(TwistAxis=(X=0,Y=0,Z=1), Input=@n35.Transform.Rotation)
   > &RigUnit_OffsetTransformForItem(
    OffsetTransform=(Rotation=(X=0,Y=0,Z=0,W=1),Translation=(X=0,Y=0,Z=0),Scale3D=(X=1,Y=1,Z=1))
    Weight=1
    bPropagateToChildren=True
    OffsetTransform.Rotation=&RigVMFunction_MathQuaternionMul(
     A=&RigVMFunction_MathQuaternionInverse(Value=@n37.Value) {template="Inverse::Execute(in Value,out Result)"}.Result
     B=@RigVMFunction_MathQuaternionSwingTwist.Twist
    ) {template="Multiply::Execute(in A,in B,out Result)"}.Result
    Item=@n36.Value
   )
   RigVMVariableNode() {variable=FK Controls}
  ]}
 RigVMModel Utils {nodes [
   RigVMFunction_UserDefinedEvent = /Script/RigVM.RigVMFunction_UserDefinedEvent(EventName=To IK) {is=struct RigVMUnitNode}
   Match IK_0 = RigVMFunctionReferenceNode(ExecuteContext=@RigVMFunction_UserDefinedEvent.ExecutePin)
   n1 = &RigVMFunction_UserDefinedEvent(EventName=To FK)
   Match FK_0 = RigVMFunctionReferenceNode(ExecuteContext=@n1.ExecutePin)
  ]}
]
hierarchy [
 Start {type=Connector, loc=<0 0 0>, rot=<0 0 0 1>}
 Root {type=Socket, loc=<0 0 0>, rot=<0 0 0 1>}
 Mid {type=Connector, loc=<0 0 0>, rot=<0 0 0 1>}
 End {type=Connector, loc=<0 0 0>, rot=<0 0 0 1>}
 Parent {type=Connector, loc=<0 0 0>, rot=<0 0 0 1>}
 IK Spaces {type=Connector, loc=<0 0 0>, rot=<0 0 0 1>}
 FK Spaces {type=Connector, loc=<0 0 0>, rot=<0 0 0 1>}
]
variables [
 Auto PV Parent {type=FRigElementKey, default=(Type=Bone,Name="")}
 Bones {type="TArray<FRigElementKey>"}
 Color {type=FLinearColor, default=(R=1,G=1,B=1,A=1)}
 Control Scale {type=double, default=1}
 Current Lengths {type="TArray<double>"}
 Debug {type=bool, default=False}
 Default FK Space Index {type=int32, default=-1}
 Default IK {type=bool, default=True}
 Default Lengths {type="TArray<double>"}
 End Bone Rotation Offset {type=FQuat, default=(X=0,Y=0,Z=0,W=1)}
 FK Control Scale Factor Profile {
  type=FRuntimeFloatCurve
  default=(EditorCurveData=(Keys=((Value=1),(Time=1,Value=1)),DefaultValue=340282346638528859811704183484516925440,PreInfinityExtrap=RCCE_Constant,PostInfinityExtrap=RCCE_Constant),ExternalCurve=None)
 }
 FK Control Shape {
  type=FRigUnit_HierarchyAddControl_ShapeSettings
  default=(bVisible=True,Name="Circle_Thick",Color=(R=1,G=1,B=1,A=1),Transform=(Rotation=(X=0,Y=-0.7071,Z=0,W=0.7071),Translation=(X=8,Y=0,Z=0),Scale3D=(X=1.7,Y=1.7,Z=1.1333)))
 }
 FK Controls {type="TArray<FRigElementKey>"}
 FK Drivers {type="TArray<FRigElementKey>"}
 FK End Display Name {type=FName, default=None}
 FK Gimbal Controls {type="TArray<FRigElementKey>"}
 FK Mid Display Name {type=FName, default=None}
 FK Mirror Behavior {type=bool, default=True}
 FK Nulls {type="TArray<FRigElementKey>", default=(())}
 FK Parent {type=FRigElementKey, default=(Type=Bone,Name="")}
 FK Rotation Offset {type=FQuat, default=(X=0,Y=0,Z=0,W=1)}
 FK Spaces {type="TArray<FRigElementKey>"}
 FK Start Display Name {type=FName, default=None}
 Gimbal Controls Vis Channel {type=FRigElementKey, default=(Type=Bone,Name="")}
 Hand Align {type=bool, default=False}
 Hand Align Channel {type=FRigElementKey, default=(Type=Bone,Name="")}
 Hand Align Default {type=bool, default=False}
 IK Base Control {type=FRigElementKey, default=(Type=None,Name="")}
 IK Base Control Vis Channel {type=FRigElementKey, default=(Type=Bone,Name="")}
 IK Compensate World Orient {type=bool, default=True}
 IK Control {type=FRigElementKey, default=(Type=None,Name="")}
 IK Control Shape {
  type=FRigUnit_HierarchyAddControl_ShapeSettings
  default=(bVisible=True,Name="Box_Thick",Color=(R=1,G=1,B=1,A=1),Transform=(Rotation=(X=0,Y=0,Z=0,W=1),Translation=(X=0,Y=0,Z=0),Scale3D=(X=0.8,Y=0.8,Z=0.8)))
 }
 IK FK Manipulation {type=bool, default=False}
 IK Gimbal Control {type=FRigElementKey, default=(Type=Bone,Name="")}
 IK Mid Display Name {type=FName, default=None}
 IK Null {type=FRigElementKey, default=(Type=Bone,Name="")}
 IK Rotation Control {type=FRigElementKey, default=(Type=Bone,Name="")}
 IK Rotation Null {type=FRigElementKey, default=(Type=Bone,Name="")}
 IK Rotation Offset {type=FQuat, default=(X=0,Y=0,Z=0,W=1)}
 IK Solve {type=bool, default=False}
 IK Spaces {type="TArray<FRigElementKey>"}
 IK Twist Null {type=FRigElementKey, default=(Type=Bone,Name="")}
 IKTransformInitial {type=FTransform, default=(Rotation=(X=0,Y=0,Z=0,W=1),Translation=(X=0,Y=0,Z=0),Scale3D=(X=1,Y=1,Z=1))}
 IKTwistInitial {type=FQuat, default=(X=0,Y=0,Z=0,W=1)}
 Ik Solve Channel {type=FRigElementKey, default=(Type=Bone,Name="")}
 Lower Segment Scale {type=float, default=1}
 Lower Segment Scale Channel {type=FRigElementKey, default=(Type=Bone,Name="")}
 Mid Control {type=FRigElementKey, default=(Type=Bone,Name="")}
 Mid Null {type=FRigElementKey, default=(Type=Bone,Name="")}
 Negative Side {type=bool, default=False}
 PV Control {type=FRigElementKey, default=(Type=None,Name="")}
 PV Control Shape {
  type=FRigUnit_HierarchyAddControl_ShapeSettings
  default=(bVisible=True,Name="Diamond_Solid",Color=(R=1,G=1,B=1,A=1),Transform=(Rotation=(X=0,Y=0,Z=0,W=1),Translation=(X=0,Y=0,Z=0),Scale3D=(X=0.4,Y=0.4,Z=0.4)))
 }
 PV Distance Scale {type=double, default=1}
 PV Twist Follow {type=double, default=0}
 PV Twist Orient Space {type=FRigElementKey, default=(Type=Bone,Name="")}
 Parent {type=FRigElementKey, default=(Type=None,Name="")}
 Parent Buffer {type=FRigElementKey, default=(Type=Bone,Name="")}
 Previous Hand Align {type=bool, default=False}
 Previous IK Solve {type=bool, default=False}
 Rotation Order {type=EEulerRotationOrder, default=ZXY}
 Sec Controls Vis Channel {type=FRigElementKey, default=(Type=Bone,Name="")}
 Segment Scale Control {type=bool, default=True}
 Softness {type=float, default=0}
 Start Twist PV {type=FRigElementKey, default=(Type=Bone,Name="")}
 Start Twist Socket {type=FRigElementKey, default=(Type=Bone,Name="")}
 Stretch {type=bool, default=False}
 Upper Segment Scale {type=float, default=1}
 Upper Segment Scale Channel {type=FRigElementKey, default=(Type=Bone,Name="")}
 Use Scale {type=bool, default=False}
 Virtual Bones {type="TArray<FRigElementKey>"}
]
